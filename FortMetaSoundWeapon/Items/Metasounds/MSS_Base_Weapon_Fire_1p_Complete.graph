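# Complete 5.1 MetaSound source for first-person weapon fire, covering single shots and bursts.
# A control subgraph triggers main, mech, sweetener, low-ammo/last-shot and suppressor layer subgraphs, which a 1P mixing subgraph combines.
# The Mobile page inlines a cheaper stereo version without sweetener.
# Output scales by master volume and feeds the AB_WeaponFire bus; a retrigger-gated delay signals OnFinished after playback ends.
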MetaSoundSource /FortMetaSoundWeapon/Items/Metasounds/MSS_Base_Weapon_Fire_1p_Complete
inputs [
 bHasUnlimitedAmmo {type=Bool}
 bIsAlwaysSuppressed {type=Bool}
 bIsFiring {type=Bool}
 BurstRate {type=Time}
 bUseSingleVoice {type=Bool}
 FireBurst {type=Trigger}
 FireRateMultiplier {type=Float}
 FireSingleShot {type=Trigger}
 LastShotOfClip.LastShotOfClipVolume {type=Float}
 LastShotOfClip.LastShotThreshold {type=Int32}
 LastShotOfClip.Sounds {type=WaveAsset:Array}
 LastShotOfClip.Volume.RandomVolumeMax {type=Float}
 LastShotOfClip.Volume.RandomVolumeMin {type=Float}
 LowAmmo.ADS.VolumeMultiply {type=Float}
 LowAmmo.Low Ammo Volume {type=Float}
 LowAmmo.Random.Pitch.Max {type=Float}
 LowAmmo.Random.Pitch.Min {type=Float}
 LowAmmo.Random.Volume.Max {type=Float}
 LowAmmo.Random.Volume.Min {type=Float}
 LowAmmo.Sounds {type=WaveAsset:Array}
 LowAmmo.Threshold {type=Int32}
 MainFire.ADS.VolumeMultiply {type=Float}
 MainFire.Random.Pitch.Max {type=Float}
 MainFire.Random.Pitch.Min {type=Float}
 MainFire.Random.Volume.Max {type=Float}
 MainFire.Random.Volume.Min {type=Float}
 MainFire.Sounds {type=WaveAsset:Array}
 MainFire.Volume {type=Float}
 MasterVolume {type=Float}
 Mech.ADS.VolumeMultiply {type=Float}
 Mech.Random.Pitch.Max {type=Float}
 Mech.Random.Pitch.Min {type=Float}
 Mech.Random.Volume.Max {type=Float}
 Mech.Random.Volume.Min {type=Float}
 Mech.Sounds {type=WaveAsset:Array}
 Mech.Volume {type=Float}
 RetriggerDelayTime {type=Time}
 RoundsPerBurst {type=Int32}
 StopBurst {type=Trigger}
 Suppressor.BaseWeaponMix {type=Float}
 Suppressor.EnvelopeDecayTime {type=Time}
 Suppressor.MasterVolume {type=Float}
 Suppressor.Random.Pitch.Max {type=Float}
 Suppressor.Random.Pitch.Min {type=Float}
 Suppressor.Random.Volume.Max {type=Float}
 Suppressor.Random.Volume.Min {type=Float}
 Suppressor.Sounds {type=WaveAsset:Array}
 Suppressor.SuppressorVolume {type=Float}
 Sweetener.ADS.VolumeMultiply {type=Float}
 Sweetener.Random.Pitch.Max {type=Float}
 Sweetener.Random.Pitch.Min {type=Float}
 Sweetener.Random.Volume.Max {type=Float}
 Sweetener.Random.Volume.Min {type=Float}
 Sweetener.Sounds {type=WaveAsset:Array}
 Sweetener.Volume {type=Float}
 UE.Source.OnPlay {type=Trigger}
 Weapon.AmmoRemaining {type=Int32}
 Weapon.BurstFireRate {type=Float}
 Weapon.ClipSize {type=Int32}
 Weapon.FireRate {type=Float}
 Weapon.FireStartTime {type=Float}
 Weapon.IsAimingDownSights {type=Bool}
 Weapon.IsSuppressorEnabled {type=Bool}
]
interfaces [Fort.Weapon:1.0, UE.OutputFormat.5dot1:1.0, UE.Source.OneShot:1.0, UE.Source:1.1]
outputs [
 bIsAlmostFinished {type=Bool}
 OnGunshotOut {type=Trigger}
 "UE.OutputFormat.5dot1.Audio:0" {type=Audio}
 "UE.OutputFormat.5dot1.Audio:1" {type=Audio}
 "UE.OutputFormat.5dot1.Audio:2" {type=Audio}
 "UE.OutputFormat.5dot1.Audio:3" {type=Audio}
 "UE.OutputFormat.5dot1.Audio:4" {type=Audio}
 "UE.OutputFormat.5dot1.Audio:5" {type=Audio}
 UE.Source.OneShot.OnFinished {type=Trigger}
]
pages [
 Default {
  is_build_graph=true
  input_defaults {
   BurstRate=0.4
   FireBurst=null
   FireRateMultiplier=1
   FireSingleShot=null
   LastShotOfClip.LastShotOfClipVolume=1
   LastShotOfClip.LastShotThreshold=1
   LastShotOfClip.Volume.RandomVolumeMax=1
   LastShotOfClip.Volume.RandomVolumeMin=0.95
   LowAmmo.ADS.VolumeMultiply=1
   LowAmmo.Low Ammo Volume=1
   LowAmmo.Random.Pitch.Max=1.05
   LowAmmo.Random.Pitch.Min=0.95
   LowAmmo.Random.Volume.Max=0.5586
   LowAmmo.Random.Volume.Min=0.95
   LowAmmo.Threshold=10
   MainFire.ADS.VolumeMultiply=1
   MainFire.Random.Pitch.Max=1.05
   MainFire.Random.Pitch.Min=0.95
   MainFire.Random.Volume.Max=1.05
   MainFire.Random.Volume.Min=0.95
   MainFire.Volume=1
   MasterVolume=1
   Mech.ADS.VolumeMultiply=1
   Mech.Random.Pitch.Max=1.05
   Mech.Random.Pitch.Min=0.95
   Mech.Random.Volume.Max=1.05
   Mech.Random.Volume.Min=0.95
   Mech.Volume=1
   RetriggerDelayTime=2
   RoundsPerBurst=2
   StopBurst=null
   Suppressor.BaseWeaponMix=0.9
   Suppressor.EnvelopeDecayTime=0.195
   Suppressor.MasterVolume=1
   Suppressor.Random.Pitch.Max=1.05
   Suppressor.Random.Pitch.Min=0.95
   Suppressor.Random.Volume.Max=1.05
   Suppressor.Random.Volume.Min=0.95
   Suppressor.SuppressorVolume=1
   Sweetener.ADS.VolumeMultiply=1
   Sweetener.Random.Pitch.Max=1.05
   Sweetener.Random.Pitch.Min=0.95
   Sweetener.Random.Volume.Max=1.05
   Sweetener.Random.Volume.Min=0.95
   Sweetener.Volume=1
   UE.Source.OnPlay=null
   Weapon.AmmoRemaining=30
   Weapon.BurstFireRate=0.1
   Weapon.ClipSize=30
   Weapon.FireRate=0.5
   Weapon.FireStartTime=0
   Weapon.IsAimingDownSights=null
   Weapon.IsSuppressorEnabled=null
   bHasUnlimitedAmmo=null
   bIsAlwaysSuppressed=null
   bIsFiring=null
   bUseSingleVoice="<MetaSoundBuilderResult.SUCCEEDED: 0>"
   LastShotOfClip.Sounds []
   LowAmmo.Sounds []
   MainFire.Sounds []
   Mech.Sounds []
   Suppressor.Sounds []
   Sweetener.Sounds []
  }
  nodes [
   n1 = InitVariable.Int32()
   n2 = /FortMetaSoundWeapon/Items/Metasounds/Subgraphs/MSP_Weapon_Fire_Control_1p(
    FireBurst=^FireBurst
    FireSingleShot=^FireSingleShot
    RoundsPerBurst=^RoundsPerBurst
    StopBurst=^StopBurst
    OnPlay=^UE.Source.OnPlay
    Weapon.AmmoRemaining=^Weapon.AmmoRemaining
    BurstRate=^Weapon.BurstFireRate
    bHasUnlimitedAmmo=^bHasUnlimitedAmmo
    bIsFiring=^bIsFiring
    Weapon.FireRate=UE.Multiply:Float(AdditionalOperands=^FireRateMultiplier, PrimaryOperand=^Weapon.FireRate).Out
   )
   n3 = VariableAccessor.Trigger(Variable=VariableMutator.Trigger(Variable=InitVariable.Trigger().Variable, Value=@n2.OutGunshot).Variable)
   n4 = VariableAccessor.Trigger(Variable=@n3.Variable)
   n5 = VariableAccessor.Trigger(Variable=@n4.Variable)
   n6 = VariableAccessor.Trigger(Variable=@n5.Variable)
   n7 = VariableAccessor.Trigger(Variable=@n6.Variable)
   n8 = /FortMetaSoundWeapon/Items/Metasounds/Subgraphs/MSP_Weapon_Fire_LowAmmo(
    LastShotOfClip.Volume.Volume=^LastShotOfClip.LastShotOfClipVolume
    LastShotOfClip.LastShotThreshold=^LastShotOfClip.LastShotThreshold
    LastShotOfClip.Sounds=^LastShotOfClip.Sounds
    LastShotOfClip.Volume.RandomVolumeMax=^LastShotOfClip.Volume.RandomVolumeMax
    LastShotOfClip.Volume.RandomVolumeMin=^LastShotOfClip.Volume.RandomVolumeMin
    LowAmmo.ADS.VolumeMultiply=^LowAmmo.ADS.VolumeMultiply
    LowAmmo.Random.Pitch.Max=^LowAmmo.Random.Pitch.Max
    LowAmmo.Random.Pitch.Min=^LowAmmo.Random.Pitch.Min
    LowAmmo.Random.Volume.Max=^LowAmmo.Random.Volume.Max
    LowAmmo.Random.Volume.Min=^LowAmmo.Random.Volume.Min
    LowAmmo.Sounds=^LowAmmo.Sounds
    LowAmmo.Threshold=^LowAmmo.Threshold
    Weapon.IsAimingDownSights=^Weapon.IsAimingDownSights
    CurrentAmmo=VariableAccessor.Int32(Variable=VariableMutator.Int32(Variable=@n1.Variable, Value=@n2.CurrentAmmo).Variable).Value
    OnGunshot=@n7.Value
   )
   n9 = /FortMetaSoundWeapon/Items/Metasounds/Subgraphs/MSP_Weapon_Fire_MainFire(
    MainFire.ADS.VolumeMultiply=^MainFire.ADS.VolumeMultiply
    MainFire.Random.Pitch.Min=^MainFire.Random.Pitch.Max
    MainFire.Random.Pitch.Max=^MainFire.Random.Pitch.Min
    MainFire.Random.Volume.Max=^MainFire.Random.Volume.Max
    MainFire.Random.Volume.Min=^MainFire.Random.Volume.Min
    MainFire.Sounds=^MainFire.Sounds
    Weapon.IsAimingDownSights=^Weapon.IsAimingDownSights
    bUseSingleVoice=^bUseSingleVoice
    OnGunshot=@n3.Value
   )
   n10 = /FortMetaSoundWeapon/Items/Metasounds/Subgraphs/MSP_Weapon_Fire_Suppressor(
    Random.Pitch.Max=^Suppressor.Random.Pitch.Max
    Random.Pitch.Min=^Suppressor.Random.Pitch.Min
    Random.Volume.Max=^Suppressor.Random.Volume.Max
    Random.Volume.Min=^Suppressor.Random.Volume.Min
    Sounds=^Suppressor.Sounds
    Weapon.IsSuppressorEnabled=^Weapon.IsSuppressorEnabled
    bUseSingleVoice=^bUseSingleVoice
    OnGunshot=@n4.Value
   )
   n11 = VariableAccessor.Trigger(Variable=@n7.Variable)
   n12 = /FortMetaSoundWeapon/Items/Metasounds/Subgraphs/MSP_Weapon_Fire_Sweetener(
    Sweetener.ADS.VolumeMultiply=^Sweetener.ADS.VolumeMultiply
    Sweetener.Random.Pitch.Max=^Sweetener.Random.Pitch.Max
    Sweetener.Random.Pitch.Min=^Sweetener.Random.Pitch.Min
    Sweetener.Random.Volume.Max=^Sweetener.Random.Volume.Max
    Sweetener.Random.Volume.Min=^Sweetener.Random.Volume.Min
    Sweetener.Sounds=^Sweetener.Sounds
    Weapon.IsAimingDownSights=^Weapon.IsAimingDownSights
    bUseSingleVoice=^bUseSingleVoice
    OnGunshot=@n11.Value
   )
   n13 = /FortMetaSoundWeapon/Items/Metasounds/Subgraphs/MSP_Weapon_Fire_Mixing_1P(
    PlaySingleShot=^FireSingleShot
    LowAmmo.Volume=^LowAmmo.Low Ammo Volume
    MainFire.Volume=^MainFire.Volume
    Mech.Volume=^Mech.Volume
    Suppressor.Mix.BaseWeapon=^Suppressor.BaseWeaponMix
    Suppressor.Envelope.DecayTime=^Suppressor.EnvelopeDecayTime
    Suppressor.MasterVolume=^Suppressor.MasterVolume
    Suppressor.Mix.Suppressor=^Suppressor.SuppressorVolume
    Sweetener.Volume=^Sweetener.Volume
    Weapon.FireRate=^Weapon.FireRate
    Weapon.IsSuppressorEnabled=^Weapon.IsSuppressorEnabled
    Weapon.IsAlwaysSuppressed=^bIsAlwaysSuppressed
    LowAmmo.C=@n8.LowAmmo.C
    MainFire.C=@n9.MainFire.C
    MainFire.L=@n9.MainFire.L
    MainFire.LFE=@n9.MainFire.LFE
    MainFire.Ls=@n9.MainFire.Ls
    MainFire.R=@n9.MainFire.R
    MainFire.Rs=@n9.MainFire.Rs
    Mech.C=/FortMetaSoundWeapon/Items/Metasounds/Subgraphs/MSP_Weapon_Fire_Mech(
     Mech.ADS.VolumeMultiply=^Mech.ADS.VolumeMultiply
     Mech.Random.Pitch.Max=^Mech.Random.Pitch.Max
     Mech.Random.Pitch.Min=^Mech.Random.Pitch.Min
     Mech.Random.Volume.Max=^Mech.Random.Volume.Max
     Mech.Random.Volume.Min=^Mech.Random.Volume.Min
     Mech.Sounds=^Mech.Sounds
     Weapon.IsAimingDownSights=^Weapon.IsAimingDownSights
     bUseSingleVoice=^bUseSingleVoice
     OnGunshot=@n6.Value
    ).Mech.L
    Suppressor.L=@n10.OutL
    Suppressor.R=@n10.OutR
    Sweetener.L=@n12.Sweetener.L
    Sweetener.R=@n12.Sweetener.R
    OnGunshot=@n5.Value
   )
   Multiply_5 = UE.Multiply:Audio by Float(AdditionalOperands=^MasterVolume, PrimaryOperand=@n13.Out.R)
   Multiply_7 = UE.Multiply:Audio by Float(AdditionalOperands=^MasterVolume, PrimaryOperand=@n13.Out.L)
   AudioBusWriter1_0 = "UE.Audio Bus Writer (1)"(In 0=UE.Add:Audio(
     AdditionalOperands=UE.Multiply:Audio by Float(PrimaryOperand=@Multiply_5.Out) {also {defaults {AdditionalOperands=0.15}}}.Out
     PrimaryOperand=UE.Multiply:Audio by Float(PrimaryOperand=@Multiply_7.Out) {also {defaults {AdditionalOperands=0.15}}}.Out
    ).Out) {also {defaults {Audio Bus=/Game/Sounds/Modulation/EnvelopeFollowers/FortWeapon/AB_WeaponFire}}}
   InitVariable.Trigger()
   InitVariable.Trigger()
   InitVariable.Trigger()
   InitVariable.Trigger()
   Multiply_0 = UE.Multiply:Audio by Float(AdditionalOperands=^MasterVolume, PrimaryOperand=@n13.Out.LFE)
   Multiply_2 = UE.Multiply:Audio by Float(AdditionalOperands=^MasterVolume, PrimaryOperand=@n13.Out.Rs)
   Multiply_3 = UE.Multiply:Audio by Float(AdditionalOperands=^MasterVolume, PrimaryOperand=@n13.Out.Ls)
   Multiply_4 = UE.Multiply:Audio by Float(AdditionalOperands=^MasterVolume, PrimaryOperand=@n13.Out.C)
   n14 = InitVariable.Trigger()
   n15 = InitVariable.Trigger()
   TriggerAccumulatorTriggerAccumulate2_0 = "TriggerAccumulator.Trigger Accumulate (2)"(
    In 1=VariableAccessor.Trigger(Variable=VariableMutator.Trigger(Variable=@n14.Variable, Value=@n8.OnFinished).Variable).Value
    In 0=VariableAccessor.Trigger(Variable=VariableMutator.Trigger(Variable=@n15.Variable, Value=@n9.OnFinished).Variable).Value
   ) {also {defaults {Auto Reset="<MetaSoundBuilderResult.SUCCEEDED: 0>"}}}
   n16 = &MSP_TriggerResolve(A=@TriggerAccumulatorTriggerAccumulate2_0.Out, B=VariableAccessor.Trigger(Variable=@n11.Variable).Value)
   TriggerDelay_0 = UE.Trigger Delay(Delay Time=^RetriggerDelayTime, In=@TriggerAccumulatorTriggerAccumulate2_0.Out)
   TriggerControl_0 = UE.Trigger Control(Open=@n16.A, Close=@n16.B, Trigger In=@TriggerDelay_0.Out)
   ValueValue_0 = Value.Value:Bool(Set=@TriggerControl_0.Trigger Out)
  ]
  links [
   {dst="@UE.OutputFormat.5dot1.Audio:3", src=Multiply_0.Out}
   {dst="@UE.OutputFormat.5dot1.Audio:5", src=Multiply_2.Out}
   {dst="@UE.OutputFormat.5dot1.Audio:4", src=Multiply_3.Out}
   {dst="@UE.OutputFormat.5dot1.Audio:2", src=Multiply_4.Out}
   {dst="@UE.OutputFormat.5dot1.Audio:1", src=Multiply_5.Out}
   {dst="@UE.OutputFormat.5dot1.Audio:0", src=Multiply_7.Out}
   {dst="@UE.Source.OneShot.OnFinished", src=TriggerControl_0.Trigger Out}
   {dst="@bIsAlmostFinished", src=ValueValue_0.Output Value}
  ]
 }
 Mobile {
  input_defaults {bIsFiring=null}
  nodes [
   n1 = &MSP_Weapon_Fire_Control_1p(
    FireBurst=^FireBurst
    FireSingleShot=^FireSingleShot
    RoundsPerBurst=^RoundsPerBurst
    StopBurst=^StopBurst
    OnPlay=^UE.Source.OnPlay
    Weapon.AmmoRemaining=^Weapon.AmmoRemaining
    BurstRate=^Weapon.BurstFireRate
    bHasUnlimitedAmmo=^bHasUnlimitedAmmo
    bIsFiring=^bIsFiring
    Weapon.FireRate=UE.Multiply:Float(AdditionalOperands=^FireRateMultiplier, PrimaryOperand=^Weapon.FireRate).Out
   )
   n2 = VariableAccessor.Int32(Variable=VariableMutator.Int32(Variable=InitVariable.Int32().Variable, Value=@n1.CurrentAmmo).Variable)
   n3 = VariableAccessor.Int32(Variable=@n2.Variable)
   n4 = VariableAccessor.Trigger(Variable=VariableMutator.Trigger(Variable=InitVariable.Trigger().Variable, Value=@n1.OutGunshot).Variable)
   n5 = VariableAccessor.Trigger(Variable=@n4.Variable)
   n6 = TriggerCompare.Clamp:Int32(B=^LowAmmo.Threshold, A=@n2.Value, Compare=@n5.Value) {also {defaults {Type=4}}}
   n7 = VariableAccessor.Trigger(Variable=VariableMutator.Trigger(Variable=InitVariable.Trigger().Variable, Value=@n6.True).Variable)
   n8 = UE.RandomFloat(Max=^LowAmmo.Random.Pitch.Max, Min=^LowAmmo.Random.Pitch.Min, Next=@n7.Value)
   n9 = UE.Multiply:Float(
    AdditionalOperands=MapRange.MapRange:Float(In Range A=Convert.Float:Int32(Int32=^LowAmmo.Threshold).Float, In=Convert.Float:Int32(Int32=@n3.Value).Float) {also {defaults {In Range B=0, Out Range A=0.3}}}.Out Value
    PrimaryOperand=UE.RandomFloat(Max=^LowAmmo.Random.Volume.Max, Min=^LowAmmo.Random.Volume.Min, Next=@n8.On Next).Value
   )
   n10 = InitVariable.Trigger()
   n11 = TriggerCompare.Clamp:Int32(B=^LastShotOfClip.LastShotThreshold, Compare=@n6.True, A=@n2.Value) {also {defaults {Type=4}}}
   ArrayRandomGet_0 = Array.Random Get:WaveAsset:Array(
    In Array=^LastShotOfClip.Sounds
    Next=VariableAccessor.Trigger(Variable=VariableMutator.Trigger(Variable=@n10.Variable, Value=@n11.True).Variable).Value
   ) {also {defaults {Enable Shared State="<MetaSoundBuilderResult.SUCCEEDED: 0>", No Repeats=-1, Weights []}}}
   FrequencyMultipliertoSemitone_1 = UE.Frequency Multiplier to Semitone:Float(Frequency Multiplier=@n8.Value)
   WavePlayer_2 = UE.Wave Player:Mono(Play=@ArrayRandomGet_0.On Next, Wave Asset=@ArrayRandomGet_0.Value, Pitch Shift=@FrequencyMultipliertoSemitone_1.Semitones) {also {version [1,1]}}
   n12 = UE.Multiply:Audio by Float(AdditionalOperands=@n9.Out, PrimaryOperand=@WavePlayer_2.Out Mono)
   n13 = VariableAccessor.Trigger(Variable=@n5.Variable)
   n14 = VariableAccessor.Trigger(Variable=@n13.Variable)
   n15 = VariableAccessor.Trigger(Variable=@n14.Variable)
   n16 = VariableAccessor.Trigger(Variable=@n15.Variable)
   n17 = UE.RandomFloat(Max=^Mech.Random.Pitch.Max, Min=^Mech.Random.Pitch.Min, Next=@n16.Value)
   ArrayRandomGet_3 = Array.Random Get:WaveAsset:Array(In Array=^Mech.Sounds, Next=@n16.Value) {also {defaults {Enable Shared State="<MetaSoundBuilderResult.SUCCEEDED: 0>", No Repeats=-1, Weights []}}}
   FrequencyMultipliertoSemitone_3 = UE.Frequency Multiplier to Semitone:Float(Frequency Multiplier=@n17.Value)
   WavePlayer_0 = UE.Wave Player:Mono(Play=@ArrayRandomGet_3.On Next, Wave Asset=@ArrayRandomGet_3.Value, Pitch Shift=@FrequencyMultipliertoSemitone_3.Semitones) {also {version [1,1]}}
   n18 = UE.Multiply:Audio by Float(
    AdditionalOperands=UE.RandomFloat(Max=^Mech.Random.Volume.Max, Min=^Mech.Random.Volume.Min, Next=@n17.On Next).Value
    PrimaryOperand=@WavePlayer_0.Out Mono
   )
   ArrayRandomGet_1 = Array.Random Get:WaveAsset:Array(In Array=^LowAmmo.Sounds, Next=@n7.Value) {also {defaults {Enable Shared State="<MetaSoundBuilderResult.SUCCEEDED: 0>", No Repeats=-1, Weights []}}}
   WavePlayer_1 = UE.Wave Player:Mono(Play=@ArrayRandomGet_1.On Next, Wave Asset=@ArrayRandomGet_1.Value, Pitch Shift=@FrequencyMultipliertoSemitone_1.Semitones) {also {version [1,1]}}
   n19 = UE.Multiply:Audio by Float(AdditionalOperands=@n9.Out, PrimaryOperand=@WavePlayer_1.Out Mono)
   n20 = Modulation.GetModulatorValue() {also {defaults {Modulator=/FortMetaSoundWeapon/Items/Modulation/CB_WeaponFire_1p_LowAmmo}}}
   n21 = UE.RandomFloat(Min=^MainFire.Random.Pitch.Max, Max=^MainFire.Random.Pitch.Min, Next=@n15.Value)
   n22 = UE.RandomFloat(Max=^MainFire.Random.Volume.Max, Min=^MainFire.Random.Volume.Min, Next=@n21.On Next)
   ArrayRandomGet_4 = Array.Random Get:WaveAsset:Array(In Array=^MainFire.Sounds, Next=@n15.Value) {also {defaults {Enable Shared State="<MetaSoundBuilderResult.SUCCEEDED: 0>", No Repeats=-1, Weights []}}}
   FrequencyMultipliertoSemitone_2 = UE.Frequency Multiplier to Semitone:Float(Frequency Multiplier=@n21.Value)
   WavePlayer_3 = UE.Wave Player:Stereo(Play=@ArrayRandomGet_4.On Next, Wave Asset=@ArrayRandomGet_4.Value, Pitch Shift=@FrequencyMultipliertoSemitone_2.Semitones) {also {version [1,1]}}
   Multiply_3 = UE.Multiply:Audio by Float(AdditionalOperands=@n22.Value, PrimaryOperand=@WavePlayer_3.Out Right)
   AudioMixerAudioMixerStereo4_0 = "AudioMixer.Audio Mixer (Stereo, 4)"(
    Gain 0=^MainFire.Volume
    In 3 L=@n12.Out
    In 3 R=@n12.Out
    In 1 L=@n18.Out
    In 1 R=@n18.Out
    In 2 L=@n19.Out
    In 2 R=@n19.Out
    Gain 2=UE.Multiply:Float(PrimaryOperand=^LowAmmo.Low Ammo Volume, AdditionalOperands=@n20.Out).Out
    Gain 1=UE.Multiply:Float(
     PrimaryOperand=^Mech.Volume
     AdditionalOperands=Modulation.GetModulatorValue() {also {defaults {Modulator=/FortMetaSoundWeapon/Items/Modulation/CB_WeaponFire_1p_Mech}}}.Out
    ).Out
    Gain 3=UE.Multiply:Float(AdditionalOperands=^LastShotOfClip.LastShotOfClipVolume, PrimaryOperand=@n20.Out).Out
    In 0 R=@Multiply_3.Out
    In 0 L=UE.Multiply:Audio by Float(AdditionalOperands=@n22.Value, PrimaryOperand=@WavePlayer_3.Out Left).Out
   )
   n23 = Modulation.GetModulatorValue() {also {defaults {Modulator=&AudioWorldization_Bus_Enclosure}}}
   ADEnvelopeADEnvelope_2 = AD Envelope.AD Envelope:Float(
    Decay Time=UE.ConversionFloatToTime(In=MapRange.MapRange:Float(In=@n23.Out) {also {defaults {Out Range A=2, Out Range B=0.9}}}.Out Value).Out
    Trigger=VariableAccessor.Trigger(Variable=@n16.Variable).Value
   ) {also {defaults {Attack Time=0, Decay Curve=0.5, Hard Reset="<MetaSoundBuilderResult.SUCCEEDED: 0>"}}}
   CrossfadeTriggerRouteFloat2_0 = "Crossfade.Trigger Route (Float, 2):Float"(
    In 0=@ADEnvelopeADEnvelope_2.Out Envelope
    Crossfade Value=MapRange.MapRange:Float(In=@n23.Out) {also {defaults {In Range A=0.9, Out Range A=1, Out Range B=0}}}.Out Value
   ) {also {defaults {In 1=1}}}
   ADEnvelopeADEnvelope_1 = AD Envelope.AD Envelope:Float(
    Trigger=^FireSingleShot
    Decay Time=UE.ConversionFloatToTime(In=UE.Multiply:Float(PrimaryOperand=^Weapon.FireRate) {also {defaults {AdditionalOperands=1.2}}}.Out).Out
   ) {also {defaults {Attack Time=0, Hard Reset="<MetaSoundBuilderResult.SUCCEEDED: 0>"}}}
   n24 = UE.Multiply:Float(
    PrimaryOperand=@CrossfadeTriggerRouteFloat2_0.Out
    AdditionalOperands=MapRange.MapRange:Float(In=@ADEnvelopeADEnvelope_1.Out Envelope) {also {defaults {Out Range A=0.9}}}.Out Value
   )
   n25 = UE.Multiply:Audio by Float(PrimaryOperand=@AudioMixerAudioMixerStereo4_0.Out L, AdditionalOperands=@n24.Out)
   n26 = UE.Multiply:Float(
    PrimaryOperand=Convert.Float:Bool(Bool=^Weapon.IsSuppressorEnabled).Float
    AdditionalOperands=UE.Subtract:Float(AdditionalOperands=Convert.Float:Bool(Bool=^bIsAlwaysSuppressed).Float) {also {defaults {PrimaryOperand=1}}}.Out
   )
   n27 = UE.Subtract:Float(AdditionalOperands=@n26.Out) {also {defaults {PrimaryOperand=1}}}
   n28 = TriggerCompare.Clamp:Bool(A=^Weapon.IsSuppressorEnabled, Compare=@n13.Value) {also {defaults {B="<MetaSoundBuilderResult.SUCCEEDED: 0>"}}}
   n29 = UE.RandomFloat(Max=^Suppressor.Random.Pitch.Max, Min=^Suppressor.Random.Pitch.Min, Next=@n28.True)
   n30 = UE.RandomFloat(Max=^Suppressor.Random.Volume.Max, Min=^Suppressor.Random.Volume.Min, Next=@n29.On Next)
   ArrayRandomGet_2 = Array.Random Get:WaveAsset:Array(In Array=^Suppressor.Sounds, Next=@n28.True) {also {defaults {Enable Shared State="<MetaSoundBuilderResult.SUCCEEDED: 0>", No Repeats=-1, Shared State Behavior=2, Weights []}}}
   FrequencyMultipliertoSemitone_0 = UE.Frequency Multiplier to Semitone:Float(Frequency Multiplier=@n29.Value)
   WavePlayer_4 = UE.Wave Player:Stereo(Play=@ArrayRandomGet_2.On Next, Wave Asset=@ArrayRandomGet_2.Value, Pitch Shift=@FrequencyMultipliertoSemitone_0.Semitones) {also {version [1,1]}}
   Multiply_0 = UE.Multiply:Audio by Float(AdditionalOperands=@n30.Value, PrimaryOperand=@WavePlayer_4.Out Right)
   ADEnvelopeADEnvelope_0 = AD Envelope.AD Envelope:Float(Decay Time=^Suppressor.EnvelopeDecayTime, Trigger=@n14.Value) {also {defaults {Attack Curve=2, Decay Curve=0.5}}}
   n31 = UE.Multiply:Audio by Float(PrimaryOperand=@AudioMixerAudioMixerStereo4_0.Out R, AdditionalOperands=@n24.Out)
   Multiply_7 = UE.Multiply:Audio by Float(AdditionalOperands=@n30.Value, PrimaryOperand=@WavePlayer_4.Out Left)
   AudioMixerAudioMixerStereo2_0 = "AudioMixer.Audio Mixer (Stereo, 2)"(
    Gain 0=^Suppressor.BaseWeaponMix
    Gain 1=^Suppressor.SuppressorVolume
    In 1 R=@Multiply_0.Out
    In 0 R=UE.Multiply:Audio by Float(AdditionalOperands=@ADEnvelopeADEnvelope_0.Out Envelope, PrimaryOperand=@n31.Out).Out
    In 0 L=UE.Multiply:Audio by Float(AdditionalOperands=@ADEnvelopeADEnvelope_0.Out Envelope, PrimaryOperand=@n25.Out).Out
    In 1 L=@Multiply_7.Out
   )
   Multiply_5 = UE.Multiply:Audio by Float(
    AdditionalOperands=@n26.Out
    PrimaryOperand=UE.Multiply:Audio by Float(AdditionalOperands=^Suppressor.MasterVolume, PrimaryOperand=@AudioMixerAudioMixerStereo2_0.Out L).Out
   )
   Multiply_11 = UE.Multiply:Audio by Float(
    AdditionalOperands=^MasterVolume
    PrimaryOperand=UE.Add:Audio(PrimaryOperand=UE.Multiply:Audio by Float(PrimaryOperand=@n25.Out, AdditionalOperands=@n27.Out).Out, AdditionalOperands=@Multiply_5.Out).Out
   )
   Multiply_2 = UE.Multiply:Audio by Float(PrimaryOperand=@Multiply_11.Out) {also {defaults {AdditionalOperands=0.15}}}
   n32 = UE.Multiply:Audio by Float(AdditionalOperands=^Suppressor.MasterVolume, PrimaryOperand=@AudioMixerAudioMixerStereo2_0.Out R)
   Multiply_4 = UE.Multiply:Audio by Float(PrimaryOperand=@n31.Out, AdditionalOperands=@n27.Out)
   Multiply_16 = UE.Multiply:Audio by Float(
    AdditionalOperands=^MasterVolume
    PrimaryOperand=UE.Add:Audio(AdditionalOperands=UE.Multiply:Audio by Float(PrimaryOperand=@n32.Out, AdditionalOperands=@n26.Out).Out, PrimaryOperand=@Multiply_4.Out).Out
   )
   AudioBusWriter1_0 = "UE.Audio Bus Writer (1)"(In 0=UE.Add:Audio(
     PrimaryOperand=@Multiply_2.Out
     AdditionalOperands=UE.Multiply:Audio by Float(PrimaryOperand=@Multiply_16.Out) {also {defaults {AdditionalOperands=0.15}}}.Out
    ).Out) {also {defaults {Audio Bus=&AB_WeaponFire}}}
   InitVariable.Trigger()
   InitVariable.Trigger()
   InitVariable.Trigger()
   InitVariable.Trigger()
   InitVariable.Trigger()
   n33 = VariableAccessor.Trigger(Variable=VariableMutator.Trigger(Variable=InitVariable.Trigger().Variable, Value=@WavePlayer_3.On Finished).Variable)
   n34 = &MSP_TriggerResolve(B=@n4.Value, A=@n33.Value)
   TriggerDelay_0 = UE.Trigger Delay(Delay Time=^RetriggerDelayTime, In=@n33.Value)
   TriggerControl_0 = UE.Trigger Control(Open=@n34.A, Close=@n34.B, Trigger In=@TriggerDelay_0.Out)
   ValueValue_0 = Value.Value:Bool(Set=@TriggerControl_0.Trigger Out)
  ]
  links [
   {dst="@UE.OutputFormat.5dot1.Audio:0", src=Multiply_11.Out}
   {dst="@UE.OutputFormat.5dot1.Audio:1", src=Multiply_16.Out}
   {dst="@UE.Source.OneShot.OnFinished", src=TriggerControl_0.Trigger Out}
   {dst="@bIsAlmostFinished", src=ValueValue_0.Output Value}
  ]
 }
]
properties {
 assetImportData=&self:AssetImportData
 numChannels=6
 outputFormat=FiveDotOne
 qualitySetting=Fort_Weapons
 sampleRate=48000
 soundClassObject=&Fort_WeaponRanged_Fire_1P_PSM
 concurrencyOverrides {maxCount=2, resolutionRule=StopOldest}
 concurrencySet [&Weapon_Gunfire_Local_Conc]
 modulationSettings {
  volumeRouting=Union
  volumeModulationDestination {bEnableModulation=true, modulators [/FortMetaSoundWeapon/Items/Modulation/PP_Weapon_Fire_1p_Vol]}
 }
}
