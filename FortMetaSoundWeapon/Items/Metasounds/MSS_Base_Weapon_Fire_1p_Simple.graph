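# Base 5.1 MetaSound source for simple first-person weapon fire, with main fire and low-ammo layers.
# A fire control subgraph turns fire, burst and ammo inputs into gunshot triggers and an ammo count for the layer subgraphs.
# Main fire feeds all six channels, while low ammo, scaled by a control bus, mixes only into center.
# It finishes after both layers end plus a retrigger delay, gated against new gunshots.
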MetaSoundSource /FortMetaSoundWeapon/Items/Metasounds/MSS_Base_Weapon_Fire_1p_Simple
inputs [
 bHasUnlimitedAmmo {type=Bool}
 bIsFiring {type=Bool}
 BurstRate {type=Time}
 FireBurst {type=Trigger}
 FireRateMultiplier {type=Float}
 FireSingleShot {type=Trigger}
 LowAmmo.ADSMainVolumeMultiply {type=Float}
 LowAmmo.Random.MaxPitch {type=Float}
 LowAmmo.Random.MaxVolume {type=Float}
 LowAmmo.Random.MinPitch {type=Float}
 LowAmmo.Random.MinVolume {type=Float}
 LowAmmo.Sounds {type=WaveAsset:Array}
 LowAmmo.Threshold {type=Int32}
 MainFire.ADSMainVolumeMultiply {type=Float}
 MainFire.Random.MaxPitch {type=Float}
 MainFire.Random.MaxVolume {type=Float}
 MainFire.Random.MinPitch {type=Float}
 MainFire.Random.MinVolume {type=Float}
 MainFire.Sounds {type=WaveAsset:Array}
 Mixing.LowAmmoVolume {type=Float}
 Mixing.MainFireVolume {type=Float}
 Mixing.WeaponOutputVolume {type=Float}
 RetriggerDelayTime {type=Time}
 RoundsPerBurst {type=Int32}
 StartFiring {type=Trigger}
 StopBurst {type=Trigger}
 StopFiring {type=Trigger}
 UE.Source.OnPlay {type=Trigger}
 UseSingleVoice {type=Bool}
 Weapon.AmmoRemaining {type=Int32}
 Weapon.BurstFireRate {type=Float}
 Weapon.ClipSize {type=Int32}
 Weapon.FireRate {type=Float}
 Weapon.FireStartTime {type=Float}
 Weapon.IsAimingDownSights {type=Bool}
 Weapon.IsSuppressorEnabled {type=Bool}
]
interfaces [Fort.Weapon:1.0, UE.OutputFormat.5dot1:1.0, UE.Source.OneShot:1.0, UE.Source:1.1]
outputs [
 bIsAlmostFinished {type=Bool}
 "UE.OutputFormat.5dot1.Audio:0" {type=Audio}
 "UE.OutputFormat.5dot1.Audio:1" {type=Audio}
 "UE.OutputFormat.5dot1.Audio:2" {type=Audio}
 "UE.OutputFormat.5dot1.Audio:3" {type=Audio}
 "UE.OutputFormat.5dot1.Audio:4" {type=Audio}
 "UE.OutputFormat.5dot1.Audio:5" {type=Audio}
 UE.Source.OneShot.OnFinished {type=Trigger}
]
pages [Default {
  input_defaults {
   BurstRate=0.4
   FireBurst=null
   FireRateMultiplier=1
   FireSingleShot=null
   LowAmmo.ADSMainVolumeMultiply=1
   LowAmmo.Random.MaxPitch=1.05
   LowAmmo.Random.MaxVolume=1.05
   LowAmmo.Random.MinPitch=0.95
   LowAmmo.Random.MinVolume=0.95
   LowAmmo.Threshold=10
   MainFire.ADSMainVolumeMultiply=1
   MainFire.Random.MaxPitch=1.05
   MainFire.Random.MaxVolume=1.05
   MainFire.Random.MinPitch=0.95
   MainFire.Random.MinVolume=0.95
   Mixing.LowAmmoVolume=1
   Mixing.MainFireVolume=1
   Mixing.WeaponOutputVolume=1
   RetriggerDelayTime=2
   RoundsPerBurst=2
   StartFiring=null
   StopBurst=null
   StopFiring=null
   UE.Source.OnPlay=null
   UseSingleVoice="<MetaSoundBuilderResult.SUCCEEDED: 0>"
   Weapon.AmmoRemaining=30
   Weapon.BurstFireRate=0.4
   Weapon.ClipSize=30
   Weapon.FireRate=0.5
   Weapon.FireStartTime=0
   Weapon.IsAimingDownSights=null
   Weapon.IsSuppressorEnabled=null
   bHasUnlimitedAmmo=null
   bIsFiring=null
   LowAmmo.Sounds []
   MainFire.Sounds []
  }
  nodes [
   InitVariable.Trigger()
   InitVariable.Trigger()
   InitVariable.Trigger()
   InitVariable.Trigger()
   InitVariable.Trigger()
   n1 = /FortMetaSoundWeapon/Items/Metasounds/Subgraphs/MSP_Weapon_Fire_Control_1p(
    FireBurst=^FireBurst
    FireSingleShot=^FireSingleShot
    RoundsPerBurst=^RoundsPerBurst
    StopBurst=^StopBurst
    OnPlay=^UE.Source.OnPlay
    Weapon.AmmoRemaining=^Weapon.AmmoRemaining
    BurstRate=^Weapon.BurstFireRate
    bHasUnlimitedAmmo=^bHasUnlimitedAmmo
    bIsFiring=^bIsFiring
    Weapon.FireRate=UE.Multiply:Float(AdditionalOperands=^FireRateMultiplier, PrimaryOperand=^Weapon.FireRate).Out
   )
   n2 = VariableAccessor.Trigger(Variable=VariableMutator.Trigger(Variable=InitVariable.Trigger().Variable, Value=@n1.OutGunshot).Variable)
   n3 = /FortMetaSoundWeapon/Items/Metasounds/Subgraphs/MSP_Weapon_Fire_MainFire(
    MainFire.ADS.VolumeMultiply=^MainFire.ADSMainVolumeMultiply
    MainFire.Random.Pitch.Max=^MainFire.Random.MaxPitch
    MainFire.Random.Volume.Max=^MainFire.Random.MaxVolume
    MainFire.Random.Pitch.Min=^MainFire.Random.MinPitch
    MainFire.Random.Volume.Min=^MainFire.Random.MinVolume
    MainFire.Sounds=^MainFire.Sounds
    bUseSingleVoice=^UseSingleVoice
    Weapon.IsAimingDownSights=^Weapon.IsAimingDownSights
    OnGunshot=@n2.Value
   )
   Multiply_0 = UE.Multiply:Audio by Float(
    AdditionalOperands=^Mixing.WeaponOutputVolume
    PrimaryOperand=UE.Multiply:Audio by Float(AdditionalOperands=^Mixing.MainFireVolume, PrimaryOperand=@n3.MainFire.Rs).Out
   )
   n4 = InitVariable.Int32()
   n5 = VariableAccessor.Trigger(Variable=@n2.Variable)
   n6 = /FortMetaSoundWeapon/Items/Metasounds/Subgraphs/MSP_Weapon_Fire_LowAmmo(
    LowAmmo.ADS.VolumeMultiply=^LowAmmo.ADSMainVolumeMultiply
    LowAmmo.Random.Pitch.Max=^LowAmmo.Random.MaxPitch
    LowAmmo.Random.Volume.Max=^LowAmmo.Random.MaxVolume
    LowAmmo.Random.Pitch.Min=^LowAmmo.Random.MinPitch
    LowAmmo.Random.Volume.Min=^LowAmmo.Random.MinVolume
    LowAmmo.Sounds=^LowAmmo.Sounds
    LowAmmo.Threshold=^LowAmmo.Threshold
    Weapon.IsAimingDownSights=^Weapon.IsAimingDownSights
    CurrentAmmo=VariableAccessor.Int32(Variable=VariableMutator.Int32(Variable=@n4.Variable, Value=@n1.CurrentAmmo).Variable).Value
    OnGunshot=@n5.Value
   )
   AudioMixerAudioMixerMono2_0 = "AudioMixer.Audio Mixer (Mono, 2)"(
    Gain 0=^Mixing.MainFireVolume
    In 1=@n6.LowAmmo.C
    Gain 1=UE.Multiply:Float(
     PrimaryOperand=^Mixing.LowAmmoVolume
     AdditionalOperands=Modulation.GetModulatorValue() {also {defaults {Modulator=/FortMetaSoundWeapon/Items/Modulation/CB_WeaponFire_1p_LowAmmo}}}.Out
    ).Out
    In 0=UE.Multiply:Audio by Float(AdditionalOperands=^Mixing.MainFireVolume, PrimaryOperand=@n3.MainFire.C).Out
   )
   Multiply_1 = UE.Multiply:Audio by Float(AdditionalOperands=^Mixing.WeaponOutputVolume, PrimaryOperand=@AudioMixerAudioMixerMono2_0.Out)
   Multiply_2 = UE.Multiply:Audio by Float(
    AdditionalOperands=^Mixing.WeaponOutputVolume
    PrimaryOperand=UE.Multiply:Audio by Float(AdditionalOperands=^Mixing.MainFireVolume, PrimaryOperand=@n3.MainFire.L).Out
   )
   Multiply_6 = UE.Multiply:Audio by Float(
    AdditionalOperands=^Mixing.WeaponOutputVolume
    PrimaryOperand=UE.Multiply:Audio by Float(AdditionalOperands=^Mixing.MainFireVolume, PrimaryOperand=@n3.MainFire.Ls).Out
   )
   Multiply_8 = UE.Multiply:Audio by Float(
    AdditionalOperands=^Mixing.WeaponOutputVolume
    PrimaryOperand=UE.Multiply:Audio by Float(AdditionalOperands=^Mixing.MainFireVolume, PrimaryOperand=@n3.MainFire.LFE).Out
   )
   Multiply_9 = UE.Multiply:Audio by Float(
    AdditionalOperands=^Mixing.WeaponOutputVolume
    PrimaryOperand=UE.Multiply:Audio by Float(AdditionalOperands=^Mixing.MainFireVolume, PrimaryOperand=@n3.MainFire.R).Out
   )
   n7 = InitVariable.Trigger()
   n8 = InitVariable.Trigger()
   TriggerAccumulatorTriggerAccumulate2_0 = "TriggerAccumulator.Trigger Accumulate (2)"(
    In 1=VariableAccessor.Trigger(Variable=VariableMutator.Trigger(Variable=@n7.Variable, Value=@n3.OnFinished).Variable).Value
    In 0=VariableAccessor.Trigger(Variable=VariableMutator.Trigger(Variable=@n8.Variable, Value=@n6.OnFinished).Variable).Value
   ) {also {defaults {Auto Reset="<MetaSoundBuilderResult.SUCCEEDED: 0>"}}}
   n9 = &MSP_TriggerResolve(A=@TriggerAccumulatorTriggerAccumulate2_0.Out, B=VariableAccessor.Trigger(Variable=@n5.Variable).Value)
   TriggerDelay_0 = UE.Trigger Delay(Delay Time=^RetriggerDelayTime, In=@TriggerAccumulatorTriggerAccumulate2_0.Out)
   TriggerControl_0 = UE.Trigger Control(Open=@n9.A, Close=@n9.B, Trigger In=@TriggerDelay_0.Out)
   ValueValue_0 = Value.Value:Bool(Set=@TriggerControl_0.Trigger Out)
  ]
  links [
   {dst="@UE.OutputFormat.5dot1.Audio:5", src=Multiply_0.Out}
   {dst="@UE.OutputFormat.5dot1.Audio:2", src=Multiply_1.Out}
   {dst="@UE.OutputFormat.5dot1.Audio:0", src=Multiply_2.Out}
   {dst="@UE.OutputFormat.5dot1.Audio:4", src=Multiply_6.Out}
   {dst="@UE.OutputFormat.5dot1.Audio:3", src=Multiply_8.Out}
   {dst="@UE.OutputFormat.5dot1.Audio:1", src=Multiply_9.Out}
   {dst="@UE.Source.OneShot.OnFinished", src=TriggerControl_0.Trigger Out}
   {dst="@bIsAlmostFinished", src=ValueValue_0.Output Value}
  ]
 }]
properties {
 assetImportData=&self:AssetImportData
 numChannels=6
 outputFormat=FiveDotOne
 qualitySetting=Fort_Weapons
 sampleRate=48000
 soundClassObject=&Fort_WeaponRanged_Fire_1P_PSM
 assetUserData [&self:MetaSoundMemberMigratorData_0]
 concurrencyOverrides {maxCount=2, resolutionRule=StopOldest}
 concurrencySet [&Weapon_Gunfire_Local_Conc]
 modulationSettings {
  volumeRouting=Union
  volumeModulationDestination {bEnableModulation=true, modulators [/FortMetaSoundWeapon/Items/Modulation/PP_Weapon_Fire_1p_Vol]}
 }
}
