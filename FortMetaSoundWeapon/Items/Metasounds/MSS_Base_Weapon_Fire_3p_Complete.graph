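# Base MetaSound source for third-person weapon fire, covering single-shot, burst and continuous firing with optional suppressor.
# A fire-control subgraph triggers close, medium and distant layers gated by distance curves, plus suppressor layers, varied by a random profile generated on play.
# A mixing subgraph applies distance, affiliation and suppressor balancing before master volume; finish fires after a retrigger delay.
# The Mobile page substitutes inline mono players and simpler mixing.
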MetaSoundSource /FortMetaSoundWeapon/Items/Metasounds/MSS_Base_Weapon_Fire_3p_Complete
inputs [
 bEnableDistanceEQ {type=Bool}
 bHasUnlimitedAmmo {type=Bool}
 bIsAlwaysSuppressed {type=Bool}
 BurstRate {type=Time}
 bUseSingleVoice {type=Bool}
 Character.Affiliation {type=Enum:FortAudioAffiliation}
 Character.Health {type=Float}
 Character.UniqueID {type=Int32}
 FireBurst {type=Trigger}
 FireRateMultiplier {type=Float}
 FireSingleShot {type=Trigger}
 MainFire.DistanceCurves {type=WaveTableBankAsset}
 MainFire.SoundsClose {type=WaveAsset:Array}
 MainFire.SoundsDistant {type=WaveAsset:Array}
 MainFire.SoundsMedium {type=WaveAsset:Array}
 MasterVolume {type=Float}
 Mixing.FalloffDistance {type=Float}
 Mixing.Friendly.FriendlyFireWeaponVolume {type=Float}
 Mixing.Suppressor.BaseWeaponVolume {type=Float}
 Mixing.Suppressor.SuppressorVolume {type=Float}
 Mixing.Volume.Close {type=Float}
 Mixing.Volume.Distant {type=Float}
 Mixing.Volume.Medium {type=Float}
 Random.EQ.CutoffFrequency.Max {type=Float}
 Random.EQ.CutoffFrequency.Min {type=Float}
 Random.EQ.Gain.Max {type=Float}
 Random.EQ.Gain.Min {type=Float}
 Random.Pitch.Max {type=Float}
 Random.Pitch.Min {type=Float}
 Random.Volume.Max {type=Float}
 Random.Volume.Min {type=Float}
 RetriggerDelayTime {type=Time}
 RoundsPerBurst {type=Int32}
 StartFiring {type=Trigger}
 StopBurst {type=Trigger}
 StopFiring {type=Trigger}
 Suppressor.EnvelopeDecay {type=Time}
 Suppressor.Sounds.Close {type=WaveAsset:Array}
 Suppressor.Sounds.Medium {type=WaveAsset:Array}
 Suppressor.Volume.Close {type=Float}
 Suppressor.Volume.Medium {type=Float}
 Targeting.FilterBandwidth {type=Float}
 Targeting.FilterFrequency {type=Float}
 Targeting.FilterGain {type=Float}
 UE.Attenuation.Distance {type=Float}
 UE.Source.OnPlay {type=Trigger}
 UE.Source.Orientation.Azimuth {type=Float}
 UE.Source.Orientation.Elevation {type=Float}
 UE.Spatialization.Azimuth {type=Float}
 UE.Spatialization.Elevation {type=Float}
 Weapon.AmmoRemaining {type=Int32}
 Weapon.BurstFireRate {type=Float}
 Weapon.ClipSize {type=Int32}
 Weapon.FireRate {type=Float}
 Weapon.FireStartTime {type=Float}
 Weapon.IsAimingDownSights {type=Bool}
 Weapon.IsSuppressorEnabled {type=Bool}
]
interfaces [
 Fort.Character:1.0
 Fort.Weapon:1.0
 UE.Attenuation:1.0
 UE.OutputFormat.Stereo:1.0
 UE.Source.OneShot:1.0
 UE.Source.Orientation:1.0
 UE.Source:1.1
 UE.Spatialization:1.0
]
outputs [
 bIsAlmostFinished {type=Bool}
 "UE.OutputFormat.Stereo.Audio:0" {type=Audio}
 "UE.OutputFormat.Stereo.Audio:1" {type=Audio}
 UE.Source.OneShot.OnFinished {type=Trigger}
]
pages [
 Default {
  is_build_graph=true
  input_defaults {
   BurstRate=0.4
   Character.Affiliation=0
   Character.Health=1
   Character.UniqueID=0
   FireBurst=null
   FireRateMultiplier=1
   FireSingleShot=null
   MainFire.DistanceCurves=&WeaponRanged_DistanceCurves
   MasterVolume=1
   Mixing.FalloffDistance=27500
   Mixing.Friendly.FriendlyFireWeaponVolume=0.7
   Mixing.Suppressor.BaseWeaponVolume=0.75
   Mixing.Suppressor.SuppressorVolume=1
   Mixing.Volume.Close=1
   Mixing.Volume.Distant=0.6
   Mixing.Volume.Medium=0.7
   Random.EQ.CutoffFrequency.Max=20000
   Random.EQ.CutoffFrequency.Min=699.9999
   Random.EQ.Gain.Max=0
   Random.EQ.Gain.Min=-3
   Random.Pitch.Max=1.05
   Random.Pitch.Min=0.95
   Random.Volume.Max=1.15
   Random.Volume.Min=0.85
   RetriggerDelayTime=2
   RoundsPerBurst=2
   StartFiring=null
   StopBurst=null
   StopFiring=null
   Suppressor.EnvelopeDecay=0.195
   Suppressor.Volume.Close=0.5
   Suppressor.Volume.Medium=0.3
   Targeting.FilterBandwidth=4
   Targeting.FilterFrequency=300
   Targeting.FilterGain=-8
   UE.Attenuation.Distance=22000
   UE.Source.OnPlay=null
   UE.Source.Orientation.Azimuth=0
   UE.Source.Orientation.Elevation=0
   UE.Spatialization.Azimuth=0
   UE.Spatialization.Elevation=-2.88
   Weapon.AmmoRemaining=30
   Weapon.BurstFireRate=0.1
   Weapon.ClipSize=30
   Weapon.FireRate=0.5
   Weapon.FireStartTime=0
   Weapon.IsAimingDownSights=null
   Weapon.IsSuppressorEnabled=null
   bEnableDistanceEQ="<MetaSoundBuilderResult.SUCCEEDED: 0>"
   bHasUnlimitedAmmo=null
   bIsAlwaysSuppressed=null
   bUseSingleVoice="<MetaSoundBuilderResult.SUCCEEDED: 0>"
   MainFire.SoundsClose []
   MainFire.SoundsDistant []
   MainFire.SoundsMedium []
   Suppressor.Sounds.Close []
   Suppressor.Sounds.Medium []
  }
  nodes [
   n1 = /FortMetaSoundWeapon/Items/Metasounds/Subgraphs/MSP_Weapon_Fire_Control_3p(
    FireBurst=^FireBurst
    FireSingleShot=^FireSingleShot
    RoundsPerBurst=^RoundsPerBurst
    StartFiring=^StartFiring
    StopBurst=^StopBurst
    StopFiring=^StopFiring
    Weapon.AmmoRemaining=^Weapon.AmmoRemaining
    BurstRate=^Weapon.BurstFireRate
    bHasUnlimitedAmmo=^bHasUnlimitedAmmo
    Weapon.FireRate=UE.Multiply:Float(AdditionalOperands=^FireRateMultiplier, PrimaryOperand=^Weapon.FireRate).Out
   )
   n2 = VariableAccessor.Trigger(Variable=VariableMutator.Trigger(Variable=InitVariable.Trigger().Variable, Value=@n1.OutGunshot).Variable)
   n3 = VariableAccessor.Trigger(Variable=@n2.Variable)
   n4 = VariableAccessor.Trigger(Variable=@n3.Variable)
   n5 = MapRange.MapRange:Float(In Range B=^Mixing.FalloffDistance, In=^UE.Attenuation.Distance)
   n6 = UE.WaveTableBankEvaluate(WaveTableBank=^MainFire.DistanceCurves, Input=UE.Reroute(Value=@n5.Out Value).Value) {also {defaults {Index=2}}}
   n7 = TriggerCompare.Clamp:Float(Compare=@n4.Value, A=@n6.Output) {also {defaults {B=0, Type=3}}}
   n8 = /FortMetaSoundWeapon/Items/Metasounds/Subgraphs/MSP_Weapon_Fire_GenerateProfile(
    Character.UniqueID=^Character.UniqueID
    Random.EQ.Frequency.Max=^Random.EQ.CutoffFrequency.Max
    Random.EQ.Frequency.Min=^Random.EQ.CutoffFrequency.Min
    Random.EQ.Gain.Max=^Random.EQ.Gain.Max
    Random.EQ.Gain.Min=^Random.EQ.Gain.Min
    Random.Pitch.Max=^Random.Pitch.Max
    Random.Pitch.Min=^Random.Pitch.Min
    Random.Volume.Max=^Random.Volume.Max
    Random.Volume.Min=^Random.Volume.Min
    Generate=^UE.Source.OnPlay
   )
   n9 = VariableAccessor.Float(Variable=VariableMutator.Float(Variable=InitVariable.Float().Variable, Value=@n8.Volume.Max).Variable)
   n10 = VariableAccessor.Float(Variable=@n9.Variable)
   n11 = VariableAccessor.Float(Variable=@n10.Variable)
   n12 = VariableAccessor.Float(Variable=VariableMutator.Float(Variable=InitVariable.Float().Variable, Value=@n8.Volume.Min).Variable)
   n13 = VariableAccessor.Float(Variable=@n12.Variable)
   n14 = VariableAccessor.Float(Variable=@n13.Variable)
   n15 = VariableAccessor.Float(Variable=VariableMutator.Float(Variable=InitVariable.Float().Variable, Value=@n8.Pitch.Max).Variable)
   n16 = VariableAccessor.Float(Variable=@n15.Variable)
   n17 = VariableAccessor.Float(Variable=VariableMutator.Float(Variable=InitVariable.Float().Variable, Value=@n8.Pitch.Min).Variable)
   n18 = VariableAccessor.Float(Variable=@n17.Variable)
   n19 = VariableAccessor.Float(Variable=@n18.Variable)
   n20 = /FortMetaSoundWeapon/Items/Metasounds/Subgraphs/MSP_Weapon_Fire_3p_Mono(
    Sounds=^MainFire.SoundsDistant
    bUseSingleVoice=^bUseSingleVoice
    OnGunshot=@n7.True
    Random.Volume.Max=@n11.Value
    Random.Volume.Min=@n14.Value
    Random.Pitch.Max=VariableAccessor.Float(Variable=@n16.Variable).Value
    Random.Pitch.Min=@n19.Value
   )
   n21 = UE.WaveTableBankEvaluate(WaveTableBank=^MainFire.DistanceCurves, Input=UE.Reroute(Value=@n5.Out Value).Value) {also {defaults {Index=1}}}
   n22 = TriggerCompare.Clamp:Float(Compare=@n3.Value, A=@n21.Output) {also {defaults {Type=3}}}
   n23 = &MSP_Weapon_Fire_3p_Mono(
    Sounds=^MainFire.SoundsMedium
    bUseSingleVoice=^bUseSingleVoice
    OnGunshot=@n22.True
    Random.Pitch.Max=@n16.Value
    Random.Volume.Min=@n13.Value
    Random.Pitch.Min=@n18.Value
    Random.Volume.Max=@n10.Value
   )
   n24 = UE.WaveTableBankEvaluate(WaveTableBank=^MainFire.DistanceCurves, Input=UE.Reroute(Value=@n5.Out Value).Value)
   n25 = TriggerCompare.Clamp:Float(Compare=@n2.Value, A=@n24.Output) {also {defaults {B=0, Type=3}}}
   n26 = /FortMetaSoundWeapon/Items/Metasounds/Subgraphs/MSP_Weapon_Fire_3p_Stereo(
    Sounds=^MainFire.SoundsClose
    bUseSingleVoice=^bUseSingleVoice
    OnGunshot=@n25.True
    Random.Pitch.Min=@n17.Value
    Random.Volume.Max=@n9.Value
    Random.Volume.Min=@n12.Value
    Random.Pitch.Max=@n15.Value
   )
   n27 = VariableAccessor.Trigger(Variable=@n4.Variable)
   n28 = /FortMetaSoundWeapon/Items/Metasounds/Subgraphs/MSP_Weapon_Fire_Suppressor_3p(
    Sounds.Close=^Suppressor.Sounds.Close
    Sounds.Medium=^Suppressor.Sounds.Medium
    Volume.Close=^Suppressor.Volume.Close
    Volume.Medium=^Suppressor.Volume.Medium
    bIsSuppressorEnabled=^Weapon.IsSuppressorEnabled
    bUseSingleVoice=^bUseSingleVoice
    Random.Volume.Min=VariableAccessor.Float(Variable=@n14.Variable).Value
    Random.Volume.Max=VariableAccessor.Float(Variable=@n11.Variable).Value
    Random.Pitch=VariableAccessor.Float(Variable=@n19.Variable).Value
    OnGunshot=@n27.Value
   )
   n29 = InitVariable.Float()
   n30 = InitVariable.Float()
   n31 = InitVariable.Float()
   n32 = /FortMetaSoundWeapon/Items/Metasounds/Subgraphs/MSP_Weapon_Fire_Mixing_3p(
    Character.Affiliation=^Character.Affiliation
    FireSingleShot=^FireSingleShot
    Mixing.Friendly=^Mixing.Friendly.FriendlyFireWeaponVolume
    Suppressor.Mixing.BaseWeaponVolume=^Mixing.Suppressor.BaseWeaponVolume
    Suppressor.Mixing.SuppressorVolume=^Mixing.Suppressor.SuppressorVolume
    Volume.Close=^Mixing.Volume.Close
    Volume.Distant=^Mixing.Volume.Distant
    Volume.Medium=^Mixing.Volume.Medium
    Suppressor.EnvelopeDecay=^Suppressor.EnvelopeDecay
    Targeting.FilterBandwidth=^Targeting.FilterBandwidth
    Targeting.FilterFrequency=^Targeting.FilterFrequency
    Targeting.FilterGain=^Targeting.FilterGain
    Orientation.Azimuth=^UE.Source.Orientation.Azimuth
    Spatialization.Elevation=^UE.Spatialization.Elevation
    Weapon.FireRate=^Weapon.FireRate
    Weapon.IsSuppressorEnabled=^Weapon.IsSuppressorEnabled
    bEnableDistanceEQ=^bEnableDistanceEQ
    Weapon.IsAlwaysSuppressed=^bIsAlwaysSuppressed
    Distant.L=@n20.Out.L
    Medium.L=@n23.Out.L
    Close.L=@n26.Out.L
    Close.R=@n26.Out.R
    Suppressor.Close.L=@n28.Close.L
    Suppressor.Close.R=@n28.Close.R
    Suppressor.Medium=@n28.Medium.L
    ProfileFilter.Gain=VariableAccessor.Float(Variable=VariableMutator.Float(Variable=@n29.Variable, Value=@n8.EQ.Gain).Variable).Value
    ProfileFilter.CutoffFrequency=VariableAccessor.Float(Variable=VariableMutator.Float(Variable=@n30.Variable, Value=@n8.EQ.CutoffFrequency).Variable).Value
    ProfileFilter.Bandwidth=VariableAccessor.Float(Variable=VariableMutator.Float(Variable=@n31.Variable, Value=@n8.EQ.Bandwidth).Variable).Value
    OnGunshot=VariableAccessor.Trigger(Variable=@n27.Variable).Value
    CurveEvaluation.Close=@n24.Output
    CurveEvaluation.Distant=@n6.Output
    CurveEvaluation.Medium=@n21.Output
   )
   Multiply_3 = UE.Multiply:Audio by Float(AdditionalOperands=^MasterVolume, PrimaryOperand=@n32.Out.L)
   n33 = VariableAccessor.Float(Variable=InitVariable.Float().Variable)
   Multiply_0 = UE.Multiply:Audio by Float(AdditionalOperands=^MasterVolume, PrimaryOperand=@n32.Out.R)
   AudioBusWriter1_0 = "UE.Audio Bus Writer (1)"(In 0=UE.Add:Audio(
     PrimaryOperand=UE.Multiply:Audio by Float(PrimaryOperand=@Multiply_3.Out, AdditionalOperands=@n33.Value).Out
     AdditionalOperands=UE.Multiply:Audio by Float(PrimaryOperand=@Multiply_0.Out, AdditionalOperands=@n33.Value).Out
    ).Out) {also {defaults {Audio Bus=/Game/Sounds/Modulation/EnvelopeFollowers/FortWeapon/AB_WeaponFire}}}
   InitVariable.Trigger()
   InitVariable.Trigger()
   InitVariable.Trigger()
   InitVariable.Trigger()
   TriggerAnyTriggerAny2_0 = "TriggerAny.Trigger Any (2)"(In 0=@n26.OnFinished, In 1=@n25.False)
   TriggerAnyTriggerAny2_1 = "TriggerAny.Trigger Any (2)"(In 0=@n20.OnFinished, In 1=@n7.False)
   TriggerAnyTriggerAny2_2 = "TriggerAny.Trigger Any (2)"(In 0=@n23.OnFinished, In 1=@n22.False)
   TriggerAccumulatorTriggerAccumulate4_0 = "TriggerAccumulator.Trigger Accumulate (4)"(In 3=@n28.OnFinished, In 0=@TriggerAnyTriggerAny2_0.Out, In 2=@TriggerAnyTriggerAny2_1.Out, In 1=@TriggerAnyTriggerAny2_2.Out) {also {defaults {Auto Reset="<MetaSoundBuilderResult.SUCCEEDED: 0>"}}}
   n34 = &MSP_TriggerResolve(A=@TriggerAccumulatorTriggerAccumulate4_0.Out, B=VariableAccessor.Trigger(Variable=@n27.Variable).Value)
   TriggerDelay_0 = UE.Trigger Delay(Delay Time=^RetriggerDelayTime, In=@TriggerAccumulatorTriggerAccumulate4_0.Out)
   TriggerControl_0 = UE.Trigger Control(Open=@n34.A, Close=@n34.B, Trigger In=@TriggerDelay_0.Out)
   ValueValue_0 = Value.Value:Bool(Set=@TriggerControl_0.Trigger Out)
   VariableMutator.Int32(Variable=InitVariable.Int32().Variable, Value=@n1.CurrentAmmo)
  ]
  links [
   {dst="@UE.OutputFormat.Stereo.Audio:1", src=Multiply_0.Out}
   {dst="@UE.OutputFormat.Stereo.Audio:0", src=Multiply_3.Out}
   {dst="@UE.Source.OneShot.OnFinished", src=TriggerControl_0.Trigger Out}
   {dst="@bIsAlmostFinished", src=ValueValue_0.Output Value}
  ]
 }
 Mobile {
  nodes [
   n1 = MapRange.MapRange:Float(In Range B=^Mixing.FalloffDistance, In=^UE.Attenuation.Distance)
   n2 = UE.WaveTableBankEvaluate(WaveTableBank=^MainFire.DistanceCurves, Input=UE.Reroute(Value=@n1.Out Value).Value) {also {defaults {Index=2}}}
   n3 = UE.Multiply:Float(AdditionalOperands=^Mixing.Volume.Distant, PrimaryOperand=@n2.Output)
   n4 = &MSP_Weapon_Fire_Control_3p(
    FireBurst=^FireBurst
    FireSingleShot=^FireSingleShot
    RoundsPerBurst=^RoundsPerBurst
    StartFiring=^StartFiring
    StopBurst=^StopBurst
    StopFiring=^StopFiring
    Weapon.AmmoRemaining=^Weapon.AmmoRemaining
    BurstRate=^Weapon.BurstFireRate
    bHasUnlimitedAmmo=^bHasUnlimitedAmmo
    Weapon.FireRate=UE.Multiply:Float(AdditionalOperands=^FireRateMultiplier, PrimaryOperand=^Weapon.FireRate).Out
   )
   n5 = VariableAccessor.Trigger(Variable=VariableMutator.Trigger(Variable=InitVariable.Trigger().Variable, Value=@n4.OutGunshot).Variable)
   n6 = VariableAccessor.Trigger(Variable=@n5.Variable)
   n7 = VariableAccessor.Trigger(Variable=@n6.Variable)
   n8 = TriggerCompare.Clamp:Float(Compare=@n7.Value, A=@n2.Output) {also {defaults {B=0, Type=3}}}
   ArrayRandomGet_3 = Array.Random Get:WaveAsset:Array(In Array=^MainFire.SoundsDistant, Next=@n8.True) {also {defaults {Enable Shared State="<MetaSoundBuilderResult.SUCCEEDED: 0>", No Repeats=-1}}}
   n9 = VariableAccessor.Float(Variable=VariableMutator.Float(
     Variable=InitVariable.Float().Variable
     Value=UE.RandomFloat(Seed=^Character.UniqueID, Next=^UE.Source.OnPlay) {also {defaults {Min=0.85}}}.Value
    ).Variable)
   n10 = VariableAccessor.Float(Variable=@n9.Variable)
   n11 = VariableAccessor.Float(Variable=@n10.Variable)
   FrequencyMultipliertoSemitone_3 = UE.Frequency Multiplier to Semitone:Float(Frequency Multiplier=@n11.Value)
   WavePlayer_1 = UE.Wave Player:Mono(Play=@ArrayRandomGet_3.On Next, Wave Asset=@ArrayRandomGet_3.Value, Pitch Shift=@FrequencyMultipliertoSemitone_3.Semitones) {also {version [1,1]}}
   n12 = UE.WaveTableBankEvaluate(WaveTableBank=^MainFire.DistanceCurves, Input=UE.Reroute(Value=@n1.Out Value).Value) {also {defaults {Index=1}}}
   n13 = UE.Multiply:Float(AdditionalOperands=^Mixing.Volume.Medium, PrimaryOperand=@n12.Output)
   n14 = TriggerCompare.Clamp:Float(Compare=@n6.Value, A=@n12.Output) {also {defaults {B=0, Type=3}}}
   ArrayRandomGet_2 = Array.Random Get:WaveAsset:Array(In Array=^MainFire.SoundsMedium, Next=@n14.True) {also {defaults {Enable Shared State="<MetaSoundBuilderResult.SUCCEEDED: 0>", No Repeats=-1, Weights []}}}
   FrequencyMultipliertoSemitone_0 = UE.Frequency Multiplier to Semitone:Float(Frequency Multiplier=@n10.Value)
   WavePlayer_4 = UE.Wave Player:Mono(Play=@ArrayRandomGet_2.On Next, Wave Asset=@ArrayRandomGet_2.Value, Pitch Shift=@FrequencyMultipliertoSemitone_0.Semitones) {also {version [1,1]}}
   n15 = UE.WaveTableBankEvaluate(WaveTableBank=^MainFire.DistanceCurves, Input=UE.Reroute(Value=@n1.Out Value).Value)
   n16 = TriggerCompare.Clamp:Float(Compare=@n5.Value, A=@n15.Output) {also {defaults {B=0, Type=3}}}
   ArrayRandomGet_0 = Array.Random Get:WaveAsset:Array(In Array=^MainFire.SoundsClose, Next=@n16.True) {also {defaults {Enable Shared State="<MetaSoundBuilderResult.SUCCEEDED: 0>", No Repeats=-1}}}
   FrequencyMultipliertoSemitone_1 = UE.Frequency Multiplier to Semitone:Float(Frequency Multiplier=@n9.Value)
   WavePlayer_3 = UE.Wave Player:Mono(Play=@ArrayRandomGet_0.On Next, Wave Asset=@ArrayRandomGet_0.Value, Pitch Shift=@FrequencyMultipliertoSemitone_1.Semitones) {also {version [1,1]}}
   Multiply_3 = UE.Multiply:Audio by Float(
    AdditionalOperands=UE.Multiply:Float(PrimaryOperand=UE.Multiply:Float(AdditionalOperands=^Mixing.Volume.Close, PrimaryOperand=@n15.Output).Out) {also {defaults {AdditionalOperands=0.65}}}.Out
    PrimaryOperand=@WavePlayer_3.Out Mono
   )
   n17 = UE.Add:Audio(
    AdditionalOperands=UE.Add:Audio(
     AdditionalOperands=UE.Multiply:Audio by Float(AdditionalOperands=@n3.Out, PrimaryOperand=@WavePlayer_1.Out Mono).Out
     PrimaryOperand=UE.Multiply:Audio by Float(AdditionalOperands=@n13.Out, PrimaryOperand=@WavePlayer_4.Out Mono).Out
    ).Out
    PrimaryOperand=@Multiply_3.Out
   )
   n18 = VariableAccessor.Trigger(Variable=@n7.Variable)
   n19 = VariableAccessor.Trigger(Variable=@n18.Variable)
   ADEnvelopeADEnvelope_0 = AD Envelope.AD Envelope:Float(Decay Time=^Suppressor.EnvelopeDecay, Trigger=VariableAccessor.Trigger(Variable=@n19.Variable).Value) {also {defaults {Attack Curve=2, Decay Curve=0.5, Hard Reset="<MetaSoundBuilderResult.SUCCEEDED: 0>"}}}
   n20 = UE.Multiply:Float(
    PrimaryOperand=^Mixing.Friendly.FriendlyFireWeaponVolume
    AdditionalOperands=Modulation.GetModulatorValue() {also {defaults {Modulator=/FortMetaSoundWeapon/Items/Modulation/CB_WeaponFire_3p_Friendly_Vol}}}.Out
   )
   CrossfadeTriggerRouteFloat4_0 = "Crossfade.Trigger Route (Float, 4):Float"(
    Crossfade Value=Convert.Float:Int32(Int32=Convert.Int32:Enum:FortAudioAffiliation("Enum:FortAudioAffiliation"=^Character.Affiliation).Int32).Float
    In 2=Modulation.GetModulatorValue() {also {defaults {Modulator=/FortMetaSoundWeapon/Items/Modulation/CB_WeaponFire_3p_Enemy_Vol}}}.Out
    In 1=@n20.Out
    In 3=@n20.Out
   ) {also {defaults {In 0=1}}}
   n21 = UE.Multiply:Float(
    PrimaryOperand=@CrossfadeTriggerRouteFloat4_0.Out
    AdditionalOperands=UE.Subtract:Float(AdditionalOperands=@n2.Output) {also {defaults {PrimaryOperand=1}}}.Out
   )
   n22 = TriggerCompare.Clamp:Bool(A=^Weapon.IsSuppressorEnabled, Compare=@n19.Value) {also {defaults {B="<MetaSoundBuilderResult.SUCCEEDED: 0>"}}}
   ArrayRandomGet_1 = Array.Random Get:WaveAsset:Array(In Array=^Suppressor.Sounds.Medium, Next=@n22.True) {also {defaults {Enable Shared State="<MetaSoundBuilderResult.SUCCEEDED: 0>", No Repeats=-1, Weights []}}}
   n23 = VariableAccessor.Float(Variable=@n11.Variable)
   FrequencyMultipliertoSemitone_2 = UE.Frequency Multiplier to Semitone:Float(Frequency Multiplier=VariableAccessor.Float(Variable=@n23.Variable).Value)
   WavePlayer_0 = UE.Wave Player:Mono(Play=@ArrayRandomGet_1.On Next, Wave Asset=@ArrayRandomGet_1.Value, Pitch Shift=@FrequencyMultipliertoSemitone_2.Semitones) {also {version [1,1]}}
   ArrayRandomGet_4 = Array.Random Get:WaveAsset:Array(In Array=^Suppressor.Sounds.Close, Next=@n22.True) {also {defaults {Enable Shared State="<MetaSoundBuilderResult.SUCCEEDED: 0>", No Repeats=-1, Weights []}}}
   FrequencyMultipliertoSemitone_4 = UE.Frequency Multiplier to Semitone:Float(Frequency Multiplier=@n23.Value)
   WavePlayer_2 = UE.Wave Player:Mono(Play=@ArrayRandomGet_4.On Next, Wave Asset=@ArrayRandomGet_4.Value, Pitch Shift=@FrequencyMultipliertoSemitone_4.Semitones) {also {version [1,1]}}
   AudioMixerAudioMixerMono3_0 = "AudioMixer.Audio Mixer (Mono, 3)"(
    In 2=@n17.Out
    Gain 1=UE.Multiply:Float(
     AdditionalOperands=^Suppressor.Volume.Medium
     PrimaryOperand=UE.Multiply:Float(AdditionalOperands=^Mixing.Suppressor.SuppressorVolume, PrimaryOperand=@n12.Output).Out
    ).Out
    Gain 0=UE.Multiply:Float(
     AdditionalOperands=UE.Multiply:Float(AdditionalOperands=^Mixing.Suppressor.SuppressorVolume, PrimaryOperand=@n15.Output).Out
     PrimaryOperand=UE.Multiply:Float(PrimaryOperand=^Suppressor.Volume.Close) {also {defaults {AdditionalOperands=0.7}}}.Out
    ).Out
    Gain 2=UE.Multiply:Float(
     AdditionalOperands=^Mixing.Suppressor.BaseWeaponVolume
     PrimaryOperand=UE.Multiply:Float(AdditionalOperands=@ADEnvelopeADEnvelope_0.Out Envelope, PrimaryOperand=@n21.Out).Out
    ).Out
    In 1=@WavePlayer_0.Out Mono
    In 0=@WavePlayer_2.Out Mono
   )
   n24 = UE.Multiply:Float(
    PrimaryOperand=Convert.Float:Bool(Bool=^Weapon.IsSuppressorEnabled).Float
    AdditionalOperands=UE.Subtract:Float(AdditionalOperands=Convert.Float:Bool(Bool=^bIsAlwaysSuppressed).Float) {also {defaults {PrimaryOperand=1}}}.Out
   )
   AudioMixerAudioMixerMono2_0 = "AudioMixer.Audio Mixer (Mono, 2)"(
    In 0=@n17.Out
    In 1=@AudioMixerAudioMixerMono3_0.Out
    Gain 0=UE.Multiply:Float(
     PrimaryOperand=@CrossfadeTriggerRouteFloat4_0.Out
     AdditionalOperands=UE.Subtract:Float(AdditionalOperands=@n24.Out) {also {defaults {PrimaryOperand=1}}}.Out
    ).Out
    Gain 1=UE.Multiply:Float(PrimaryOperand=@CrossfadeTriggerRouteFloat4_0.Out, AdditionalOperands=@n24.Out).Out
   )
   Multiply_5 = UE.Multiply:Audio by Float(AdditionalOperands=^MasterVolume, PrimaryOperand=@AudioMixerAudioMixerMono2_0.Out)
   n25 = VariableAccessor.Float(Variable=InitVariable.Float().Variable)
   Multiply_0 = UE.Multiply:Audio by Float(AdditionalOperands=^MasterVolume, PrimaryOperand=@AudioMixerAudioMixerMono2_0.Out)
   AudioBusWriter1_0 = "UE.Audio Bus Writer (1)"(In 0=UE.Add:Audio(
     PrimaryOperand=UE.Multiply:Audio by Float(PrimaryOperand=@Multiply_5.Out, AdditionalOperands=@n25.Value).Out
     AdditionalOperands=UE.Multiply:Audio by Float(PrimaryOperand=@Multiply_0.Out, AdditionalOperands=@n25.Value).Out
    ).Out) {also {defaults {Audio Bus=&AB_WeaponFire}}}
   TriggerAnyTriggerAny2_0 = "TriggerAny.Trigger Any (2)"(In 1=@n16.False, In 0=@WavePlayer_3.On Finished)
   TriggerAnyTriggerAny2_1 = "TriggerAny.Trigger Any (2)"(In 1=@n8.False, In 0=@WavePlayer_1.On Finished)
   TriggerAnyTriggerAny2_2 = "TriggerAny.Trigger Any (2)"(In 1=@n14.False, In 0=@WavePlayer_4.On Finished)
   TriggerAnyTriggerAny2_3 = "TriggerAny.Trigger Any (2)"(In 1=@n22.False, In 0=@WavePlayer_2.On Finished)
   TriggerAccumulatorTriggerAccumulate4_0 = "TriggerAccumulator.Trigger Accumulate (4)"(In 0=@TriggerAnyTriggerAny2_0.Out, In 2=@TriggerAnyTriggerAny2_1.Out, In 1=@TriggerAnyTriggerAny2_2.Out, In 3=@TriggerAnyTriggerAny2_3.Out) {also {defaults {Auto Reset="<MetaSoundBuilderResult.SUCCEEDED: 0>"}}}
   n26 = &MSP_TriggerResolve(A=@TriggerAccumulatorTriggerAccumulate4_0.Out, B=@n18.Value)
   TriggerDelay_0 = UE.Trigger Delay(Delay Time=^RetriggerDelayTime, In=@TriggerAccumulatorTriggerAccumulate4_0.Out)
   TriggerControl_0 = UE.Trigger Control(Open=@n26.A, Close=@n26.B, Trigger In=@TriggerDelay_0.Out)
   ValueValue_0 = Value.Value:Bool(Set=@TriggerControl_0.Trigger Out)
   VariableMutator.Int32(Variable=InitVariable.Int32().Variable, Value=@n4.CurrentAmmo)
  ]
  links [
   {dst="@UE.OutputFormat.Stereo.Audio:1", src=Multiply_0.Out}
   {dst="@UE.OutputFormat.Stereo.Audio:0", src=Multiply_5.Out}
   {dst="@UE.Source.OneShot.OnFinished", src=TriggerControl_0.Trigger Out}
   {dst="@bIsAlmostFinished", src=ValueValue_0.Output Value}
  ]
 }
]
properties {
 assetImportData=&self:AssetImportData
 attenuationSettings=&Fort_Weapon_Projectile_NonAtt_Athena_Att
 maxDistance=27500
 numChannels=2
 outputFormat=Stereo
 qualitySetting=Fort_Weapons
 sampleRate=48000
 soundClassObject=&Fort_WeaponRanged_Fire_3P_PSM
 concurrencyOverrides {bLimitToOwner=true, maxCount=2, resolutionRule=StopOldest}
 concurrencySet [&Weapon_Gunfire_Local_Conc, &Weapon_Gunfire_Global_Conc]
 modulationSettings {
  volumeRouting=Union
  volumeModulationDestination {bEnableModulation=true, modulators [/FortMetaSoundWeapon/Items/Modulation/PP_Weapon_Fire_3p_Vol]}
 }
}
