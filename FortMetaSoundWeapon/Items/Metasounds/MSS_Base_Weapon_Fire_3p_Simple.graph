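# Base stereo MetaSound source for simple third-person weapon fire, turning fire, burst and ammo inputs into gunshot triggers through a control subgraph.
# Distance curves gate and mix stereo close with mono medium and distant layers, then affiliation-based modulation sets volume.
# The Default page adds per-character randomized pitch, volume and EQ; Mobile drops EQ and simplifies randomization.
# The source finishes after all layers end and a retrigger delay passes without another shot.
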MetaSoundSource /FortMetaSoundWeapon/Items/Metasounds/MSS_Base_Weapon_Fire_3p_Simple
inputs [
 bHasUnlimitedAmmo {type=Bool}
 bUseSingleVoice {type=Bool}
 Character.Affiliation {type=Enum:FortAudioAffiliation}
 Character.Health {type=Float}
 Character.UniqueID {type=Int32}
 FireBurst {type=Trigger}
 FireRateMultiplier {type=Float}
 FireSingleShot {type=Trigger}
 MainFire.DistanceCurves {type=WaveTableBankAsset}
 MainFire.FalloffDistance {type=Float}
 MainFire.SoundsClose {type=WaveAsset:Array}
 MainFire.SoundsDistant {type=WaveAsset:Array}
 MainFire.SoundsMedium {type=WaveAsset:Array}
 Mixing.CloseGain {type=Float}
 Mixing.DistantGain {type=Float}
 Mixing.FalloffDistance {type=Float}
 Mixing.Friendly.FriendlyFireWeaponVolume {type=Float}
 Mixing.MediumGain {type=Float}
 Mixing.WeaponOutputVolume {type=Float}
 Random.EQ.A.MaxFreq {type=Float}
 Random.EQ.A.MaxGain {type=Float}
 Random.EQ.A.MinFreq {type=Float}
 Random.EQ.A.MinGain {type=Float}
 Random.MaxPitch {type=Float}
 Random.MaxVolume {type=Float}
 Random.MinPitch {type=Float}
 Random.MinVolume {type=Float}
 RetriggerDelayTime {type=Time}
 RoundsPerBurst {type=Int32}
 StartFiring {type=Trigger}
 StopBurst {type=Trigger}
 StopFiring {type=Trigger}
 UE.Attenuation.Distance {type=Float}
 UE.Source.OnPlay {type=Trigger}
 Weapon.AmmoRemaining {type=Int32}
 Weapon.BurstFireRate {type=Float}
 Weapon.ClipSize {type=Int32}
 Weapon.FireRate {type=Float}
 Weapon.FireStartTime {type=Float}
 Weapon.IsAimingDownSights {type=Bool}
 Weapon.IsSuppressorEnabled {type=Bool}
]
interfaces [Fort.Character:1.0, Fort.Weapon:1.0, UE.Attenuation:1.0, UE.OutputFormat.Stereo:1.0, UE.Source.OneShot:1.0, UE.Source:1.1]
outputs [
 bIsAlmostFinished {type=Bool}
 "UE.OutputFormat.Stereo.Audio:0" {type=Audio}
 "UE.OutputFormat.Stereo.Audio:1" {type=Audio}
 UE.Source.OneShot.OnFinished {type=Trigger}
]
pages [
 Default {
  is_build_graph=true
  input_defaults {
   Character.Affiliation=0
   Character.Health=1
   Character.UniqueID=0
   FireBurst=null
   FireRateMultiplier=1
   FireSingleShot=null
   MainFire.DistanceCurves=&WeaponRanged_DistanceCurves
   MainFire.FalloffDistance=27500
   Mixing.CloseGain=1
   Mixing.DistantGain=0.6
   Mixing.FalloffDistance=27500
   Mixing.Friendly.FriendlyFireWeaponVolume=0.7
   Mixing.MediumGain=0.7
   Mixing.WeaponOutputVolume=1
   Random.EQ.A.MaxFreq=20000
   Random.EQ.A.MaxGain=0
   Random.EQ.A.MinFreq=699.9999
   Random.EQ.A.MinGain=-3
   Random.MaxPitch=1.15
   Random.MaxVolume=1.15
   Random.MinPitch=0.95
   Random.MinVolume=0.85
   RetriggerDelayTime=2
   RoundsPerBurst=2
   StartFiring=null
   StopBurst=null
   StopFiring=null
   UE.Attenuation.Distance=25000
   UE.Source.OnPlay=null
   Weapon.AmmoRemaining=30
   Weapon.BurstFireRate=0.4
   Weapon.ClipSize=30
   Weapon.FireRate=0.5
   Weapon.FireStartTime=0
   Weapon.IsAimingDownSights=null
   Weapon.IsSuppressorEnabled=null
   bHasUnlimitedAmmo=null
   bUseSingleVoice="<MetaSoundBuilderResult.SUCCEEDED: 0>"
   MainFire.SoundsClose []
   MainFire.SoundsDistant []
   MainFire.SoundsMedium []
  }
  nodes [
   InitVariable.Trigger()
   InitVariable.Trigger()
   InitVariable.Trigger()
   InitVariable.Trigger()
   InitVariable.Trigger()
   InitVariable.Trigger()
   InitVariable.Trigger()
   n1 = /FortMetaSoundWeapon/Items/Metasounds/Subgraphs/MSP_Weapon_Fire_Control_3p(
    FireBurst=^FireBurst
    FireSingleShot=^FireSingleShot
    RoundsPerBurst=^RoundsPerBurst
    StartFiring=^StartFiring
    StopBurst=^StopBurst
    StopFiring=^StopFiring
    Weapon.AmmoRemaining=^Weapon.AmmoRemaining
    BurstRate=^Weapon.BurstFireRate
    bHasUnlimitedAmmo=^bHasUnlimitedAmmo
    Weapon.FireRate=UE.Multiply:Float(AdditionalOperands=^FireRateMultiplier, PrimaryOperand=^Weapon.FireRate).Out
   )
   n2 = VariableAccessor.Trigger(Variable=VariableMutator.Trigger(Variable=InitVariable.Trigger().Variable, Value=@n1.OutGunshot).Variable)
   n3 = VariableAccessor.Trigger(Variable=@n2.Variable)
   n4 = VariableAccessor.Trigger(Variable=@n3.Variable)
   n5 = MapRange.MapRange:Float(In Range B=^Mixing.FalloffDistance, In=^UE.Attenuation.Distance)
   n6 = UE.WaveTableBankEvaluate(WaveTableBank=^MainFire.DistanceCurves, Input=UE.Reroute(Value=@n5.Out Value).Value) {also {defaults {Index=2}}}
   n7 = TriggerCompare.Clamp:Float(Compare=@n4.Value, A=@n6.Output) {also {defaults {Type=3}}}
   n8 = /FortMetaSoundWeapon/Items/Metasounds/Subgraphs/MSP_Weapon_Fire_GenerateProfile(
    Character.UniqueID=^Character.UniqueID
    Random.EQ.Frequency.Max=^Random.EQ.A.MaxFreq
    Random.EQ.Gain.Max=^Random.EQ.A.MaxGain
    Random.EQ.Frequency.Min=^Random.EQ.A.MinFreq
    Random.EQ.Gain.Min=^Random.EQ.A.MinGain
    Random.Pitch.Max=^Random.MaxPitch
    Random.Volume.Max=^Random.MaxVolume
    Random.Pitch.Min=^Random.MinPitch
    Random.Volume.Min=^Random.MinVolume
    Generate=^UE.Source.OnPlay
   )
   n9 = VariableAccessor.Float(Variable=VariableMutator.Float(Variable=InitVariable.Float().Variable, Value=@n8.Volume.Max).Variable)
   n10 = VariableAccessor.Float(Variable=@n9.Variable)
   n11 = VariableAccessor.Float(Variable=VariableMutator.Float(Variable=InitVariable.Float().Variable, Value=@n8.Pitch.Max).Variable)
   n12 = VariableAccessor.Float(Variable=@n11.Variable)
   n13 = VariableAccessor.Float(Variable=VariableMutator.Float(Variable=InitVariable.Float().Variable, Value=@n8.Volume.Min).Variable)
   n14 = VariableAccessor.Float(Variable=@n13.Variable)
   n15 = VariableAccessor.Float(Variable=VariableMutator.Float(Variable=InitVariable.Float().Variable, Value=@n8.Pitch.Min).Variable)
   n16 = VariableAccessor.Float(Variable=@n15.Variable)
   n17 = /FortMetaSoundWeapon/Items/Metasounds/Subgraphs/MSP_Weapon_Fire_3p_Mono(
    Sounds=^MainFire.SoundsDistant
    bUseSingleVoice=^bUseSingleVoice
    OnGunshot=@n7.True
    Random.Volume.Max=VariableAccessor.Float(Variable=@n10.Variable).Value
    Random.Pitch.Max=VariableAccessor.Float(Variable=@n12.Variable).Value
    Random.Volume.Min=VariableAccessor.Float(Variable=@n14.Variable).Value
    Random.Pitch.Min=VariableAccessor.Float(Variable=@n16.Variable).Value
   )
   n18 = UE.WaveTableBankEvaluate(WaveTableBank=^MainFire.DistanceCurves, Input=UE.Reroute(Value=@n5.Out Value).Value) {also {defaults {Index=1}}}
   n19 = TriggerCompare.Clamp:Float(Compare=@n3.Value, A=@n18.Output) {also {defaults {Type=3}}}
   n20 = &MSP_Weapon_Fire_3p_Mono(
    Sounds=^MainFire.SoundsMedium
    bUseSingleVoice=^bUseSingleVoice
    OnGunshot=@n19.True
    Random.Volume.Max=@n10.Value
    Random.Volume.Min=@n14.Value
    Random.Pitch.Min=@n16.Value
    Random.Pitch.Max=@n11.Value
   )
   AudioMixerAudioMixerMono2_0 = "AudioMixer.Audio Mixer (Mono, 2)"(
    In 1=@n17.Out.L
    In 0=@n20.Out.L
    Gain 1=UE.Multiply:Float(AdditionalOperands=^Mixing.DistantGain, PrimaryOperand=@n6.Output).Out
    Gain 0=UE.Multiply:Float(AdditionalOperands=^Mixing.MediumGain, PrimaryOperand=@n18.Output).Out
   )
   n21 = UE.WaveTableBankEvaluate(WaveTableBank=^MainFire.DistanceCurves, Input=UE.Reroute(Value=@n5.Out Value).Value)
   n22 = TriggerCompare.Clamp:Float(Compare=@n2.Value, A=@n21.Output) {also {defaults {Type=3}}}
   n23 = /FortMetaSoundWeapon/Items/Metasounds/Subgraphs/MSP_Weapon_Fire_3p_Stereo(
    Sounds=^MainFire.SoundsClose
    bUseSingleVoice=^bUseSingleVoice
    OnGunshot=@n22.True
    Random.Pitch.Max=@n12.Value
    Random.Volume.Min=@n13.Value
    Random.Volume.Max=@n9.Value
    Random.Pitch.Min=@n15.Value
   )
   n24 = UE.Multiply:Float(AdditionalOperands=^Mixing.CloseGain, PrimaryOperand=@n21.Output)
   n25 = InitVariable.Float()
   n26 = Clamp.Clamp:Float(In=VariableAccessor.Float(Variable=VariableMutator.Float(Variable=@n25.Variable, Value=@n8.EQ.CutoffFrequency).Variable).Value) {also {defaults {Max=20000, Min=20}}}
   n27 = VariableAccessor.Float(Variable=VariableMutator.Float(Variable=InitVariable.Float().Variable, Value=@n8.EQ.Gain).Variable)
   n28 = VariableAccessor.Float(Variable=VariableMutator.Float(Variable=InitVariable.Float().Variable, Value=@n8.EQ.Bandwidth).Variable)
   BiquadFilter_1 = UE.Biquad Filter:Audio(
    In=UE.Add:Audio(
     AdditionalOperands=@AudioMixerAudioMixerMono2_0.Out
     PrimaryOperand=UE.Multiply:Audio by Float(PrimaryOperand=@n23.Out.R, AdditionalOperands=@n24.Out).Out
    ).Out
    Cutoff Frequency=@n26.Value
    Gain=@n27.Value
    Bandwidth=@n28.Value
   ) {also {defaults {Type=4}, version [2,0]}}
   n29 = UE.Multiply:Float(
    PrimaryOperand=^Mixing.Friendly.FriendlyFireWeaponVolume
    AdditionalOperands=Modulation.GetModulatorValue() {also {defaults {Modulator=/FortMetaSoundWeapon/Items/Modulation/CB_WeaponFire_3p_Friendly_Vol}}}.Out
   )
   CrossfadeTriggerRouteFloat4_0 = "Crossfade.Trigger Route (Float, 4):Float"(
    Crossfade Value=Convert.Float:Int32(Int32=Convert.Int32:Enum:FortAudioAffiliation("Enum:FortAudioAffiliation"=^Character.Affiliation).Int32).Float
    In 2=Modulation.GetModulatorValue() {also {defaults {Modulator=/FortMetaSoundWeapon/Items/Modulation/CB_WeaponFire_3p_Enemy_Vol}}}.Out
    In 1=@n29.Out
    In 3=@n29.Out
   ) {also {defaults {In 0=1}}}
   n30 = UE.Multiply:Float(AdditionalOperands=^Mixing.WeaponOutputVolume, PrimaryOperand=@CrossfadeTriggerRouteFloat4_0.Out)
   Multiply_1 = UE.Multiply:Audio by Float(PrimaryOperand=@BiquadFilter_1.Out, AdditionalOperands=@n30.Out)
   BiquadFilter_0 = UE.Biquad Filter:Audio(
    In=UE.Add:Audio(
     AdditionalOperands=@AudioMixerAudioMixerMono2_0.Out
     PrimaryOperand=UE.Multiply:Audio by Float(PrimaryOperand=@n23.Out.L, AdditionalOperands=@n24.Out).Out
    ).Out
    Cutoff Frequency=@n26.Value
    Gain=@n27.Value
    Bandwidth=@n28.Value
   ) {also {defaults {Type=4}, version [2,0]}}
   Multiply_2 = UE.Multiply:Audio by Float(PrimaryOperand=@BiquadFilter_0.Out, AdditionalOperands=@n30.Out)
   TriggerAnyTriggerAny2_0 = "TriggerAny.Trigger Any (2)"(In 0=@n23.OnFinished, In 1=@n22.False)
   TriggerAnyTriggerAny2_1 = "TriggerAny.Trigger Any (2)"(In 0=@n17.OnFinished, In 1=@n7.False)
   TriggerAnyTriggerAny2_2 = "TriggerAny.Trigger Any (2)"(In 0=@n20.OnFinished, In 1=@n19.False)
   TriggerAccumulatorTriggerAccumulate3_0 = "TriggerAccumulator.Trigger Accumulate (3)"(In 0=@TriggerAnyTriggerAny2_0.Out, In 2=@TriggerAnyTriggerAny2_1.Out, In 1=@TriggerAnyTriggerAny2_2.Out) {also {defaults {Auto Reset="<MetaSoundBuilderResult.SUCCEEDED: 0>"}}}
   n31 = &MSP_TriggerResolve(A=@TriggerAccumulatorTriggerAccumulate3_0.Out, B=VariableAccessor.Trigger(Variable=@n4.Variable).Value)
   TriggerDelay_0 = UE.Trigger Delay(Delay Time=^RetriggerDelayTime, In=@TriggerAccumulatorTriggerAccumulate3_0.Out)
   TriggerControl_0 = UE.Trigger Control(Open=@n31.A, Close=@n31.B, Trigger In=@TriggerDelay_0.Out)
   ValueValue_0 = Value.Value:Bool(Set=@TriggerControl_0.Trigger Out)
   VariableMutator.Int32(Variable=InitVariable.Int32().Variable, Value=@n1.CurrentAmmo)
  ]
  links [
   {dst="@UE.OutputFormat.Stereo.Audio:1", src=Multiply_1.Out}
   {dst="@UE.OutputFormat.Stereo.Audio:0", src=Multiply_2.Out}
   {dst="@UE.Source.OneShot.OnFinished", src=TriggerControl_0.Trigger Out}
   {dst="@bIsAlmostFinished", src=ValueValue_0.Output Value}
  ]
 }
 Mobile {
  input_defaults {bUseSingleVoice="<MetaSoundBuilderResult.SUCCEEDED: 0>"}
  nodes [
   n1 = &MSP_Weapon_Fire_Control_3p(
    FireBurst=^FireBurst
    FireSingleShot=^FireSingleShot
    RoundsPerBurst=^RoundsPerBurst
    StartFiring=^StartFiring
    StopBurst=^StopBurst
    StopFiring=^StopFiring
    Weapon.AmmoRemaining=^Weapon.AmmoRemaining
    BurstRate=^Weapon.BurstFireRate
    Weapon.FireRate=^Weapon.FireRate
    bHasUnlimitedAmmo=^bHasUnlimitedAmmo
   )
   n2 = VariableAccessor.Trigger(Variable=VariableMutator.Trigger(Variable=InitVariable.Trigger().Variable, Value=@n1.OutGunshot).Variable)
   n3 = VariableAccessor.Trigger(Variable=@n2.Variable)
   n4 = MapRange.MapRange:Float(In Range B=^Mixing.FalloffDistance, In=^UE.Attenuation.Distance)
   n5 = UE.WaveTableBankEvaluate(WaveTableBank=^MainFire.DistanceCurves, Input=UE.Reroute(Value=@n4.Out Value).Value) {also {defaults {Index=2}}}
   n6 = TriggerCompare.Clamp:Float(Compare=@n3.Value, A=@n5.Output) {also {defaults {Type=3}}}
   n7 = VariableAccessor.Float(Variable=VariableMutator.Float(
     Variable=InitVariable.Float().Variable
     Value=UE.RandomFloat(Seed=^Character.UniqueID, Next=^UE.Source.OnPlay) {also {defaults {Min=0.85}}}.Value
    ).Variable)
   n8 = VariableAccessor.Float(Variable=@n7.Variable)
   n9 = VariableAccessor.Float(Variable=@n8.Variable)
   n10 = &MSP_Weapon_Fire_3p_Mono(
    Sounds=^MainFire.SoundsDistant
    bUseSingleVoice=^bUseSingleVoice
    OnGunshot=@n6.True
    Random.Pitch.Min=@n9.Value
    Random.Volume.Min=@n9.Value
   ) {also {defaults {Random.Pitch.Max=1.05, Random.Volume.Max=1.15}}}
   n11 = VariableAccessor.Trigger(Variable=@n3.Variable)
   n12 = UE.WaveTableBankEvaluate(WaveTableBank=^MainFire.DistanceCurves, Input=UE.Reroute(Value=@n4.Out Value).Value) {also {defaults {Index=1}}}
   n13 = TriggerCompare.Clamp:Float(Compare=@n11.Value, A=@n12.Output) {also {defaults {Type=3}}}
   n14 = &MSP_Weapon_Fire_3p_Mono(
    Sounds=^MainFire.SoundsMedium
    bUseSingleVoice=^bUseSingleVoice
    OnGunshot=@n13.True
    Random.Pitch.Min=@n8.Value
    Random.Volume.Min=@n8.Value
   ) {also {defaults {Random.Pitch.Max=1.05, Random.Volume.Max=1.15}}}
   AudioMixerAudioMixerMono2_0 = "AudioMixer.Audio Mixer (Mono, 2)"(
    In 1=@n10.Out.L
    In 0=@n14.Out.L
    Gain 1=UE.Multiply:Float(AdditionalOperands=^Mixing.DistantGain, PrimaryOperand=@n5.Output).Out
    Gain 0=UE.Multiply:Float(AdditionalOperands=^Mixing.MediumGain, PrimaryOperand=@n12.Output).Out
   )
   n15 = UE.WaveTableBankEvaluate(WaveTableBank=^MainFire.DistanceCurves, Input=UE.Reroute(Value=@n4.Out Value).Value)
   n16 = TriggerCompare.Clamp:Float(Compare=VariableAccessor.Trigger(Variable=@n11.Variable).Value, A=@n15.Output) {also {defaults {Type=3}}}
   n17 = &MSP_Weapon_Fire_3p_Stereo(Sounds=^MainFire.SoundsClose, bUseSingleVoice=^bUseSingleVoice, OnGunshot=@n16.True, Random.Pitch.Min=@n7.Value, Random.Volume.Min=@n7.Value) {also {defaults {Random.Pitch.Max=1.05, Random.Volume.Max=1.15}}}
   n18 = UE.Multiply:Float(AdditionalOperands=^Mixing.CloseGain, PrimaryOperand=@n15.Output)
   n19 = UE.Multiply:Float(
    PrimaryOperand=^Mixing.Friendly.FriendlyFireWeaponVolume
    AdditionalOperands=Modulation.GetModulatorValue() {also {defaults {Modulator=&CB_WeaponFire_3p_Friendly_Vol}}}.Out
   )
   CrossfadeTriggerRouteFloat4_0 = "Crossfade.Trigger Route (Float, 4):Float"(
    Crossfade Value=Convert.Float:Int32(Int32=Convert.Int32:Enum:FortAudioAffiliation("Enum:FortAudioAffiliation"=^Character.Affiliation).Int32).Float
    In 2=Modulation.GetModulatorValue() {also {defaults {Modulator=&CB_WeaponFire_3p_Enemy_Vol}}}.Out
    In 1=@n19.Out
    In 3=@n19.Out
   ) {also {defaults {In 0=1}}}
   n20 = UE.Multiply:Float(PrimaryOperand=^Mixing.WeaponOutputVolume, AdditionalOperands=@CrossfadeTriggerRouteFloat4_0.Out)
   Multiply_1 = UE.Multiply:Audio by Float(
    PrimaryOperand=UE.Add:Audio(
     AdditionalOperands=@AudioMixerAudioMixerMono2_0.Out
     PrimaryOperand=UE.Multiply:Audio by Float(PrimaryOperand=@n17.Out.R, AdditionalOperands=@n18.Out).Out
    ).Out
    AdditionalOperands=@n20.Out
   )
   Multiply_2 = UE.Multiply:Audio by Float(
    PrimaryOperand=UE.Add:Audio(
     AdditionalOperands=@AudioMixerAudioMixerMono2_0.Out
     PrimaryOperand=UE.Multiply:Audio by Float(PrimaryOperand=@n17.Out.L, AdditionalOperands=@n18.Out).Out
    ).Out
    AdditionalOperands=@n20.Out
   )
   TriggerAnyTriggerAny2_0 = "TriggerAny.Trigger Any (2)"(In 0=@n14.OnFinished, In 1=@n13.False)
   TriggerAnyTriggerAny2_1 = "TriggerAny.Trigger Any (2)"(In 0=@n10.OnFinished, In 1=@n6.False)
   TriggerAnyTriggerAny2_2 = "TriggerAny.Trigger Any (2)"(In 0=@n17.OnFinished, In 1=@n16.False)
   TriggerAccumulatorTriggerAccumulate3_0 = "TriggerAccumulator.Trigger Accumulate (3)"(In 1=@TriggerAnyTriggerAny2_0.Out, In 2=@TriggerAnyTriggerAny2_1.Out, In 0=@TriggerAnyTriggerAny2_2.Out) {also {defaults {Auto Reset="<MetaSoundBuilderResult.SUCCEEDED: 0>"}}}
   n21 = &MSP_TriggerResolve(A=@TriggerAccumulatorTriggerAccumulate3_0.Out, B=@n2.Value)
   TriggerDelay_0 = UE.Trigger Delay(Delay Time=^RetriggerDelayTime, In=@TriggerAccumulatorTriggerAccumulate3_0.Out)
   TriggerControl_0 = UE.Trigger Control(Open=@n21.A, Close=@n21.B, Trigger In=@TriggerDelay_0.Out)
   ValueValue_0 = Value.Value:Bool(Set=@TriggerControl_0.Trigger Out)
   VariableMutator.Int32(Variable=InitVariable.Int32().Variable, Value=@n1.CurrentAmmo)
  ]
  links [
   {dst="@UE.OutputFormat.Stereo.Audio:1", src=Multiply_1.Out}
   {dst="@UE.OutputFormat.Stereo.Audio:0", src=Multiply_2.Out}
   {dst="@UE.Source.OneShot.OnFinished", src=TriggerControl_0.Trigger Out}
   {dst="@bIsAlmostFinished", src=ValueValue_0.Output Value}
  ]
 }
]
properties {
 assetImportData=&self:AssetImportData
 attenuationSettings=&Fort_Weapon_Projectile_NonAtt_Athena_Att
 maxDistance=27500
 numChannels=2
 outputFormat=Stereo
 qualitySetting=Fort_Weapons
 sampleRate=48000
 soundClassObject=&Fort_WeaponRanged_Fire_3P_PSM
 assetUserData [&self:MetaSoundMemberMigratorData_0]
 concurrencyOverrides {bLimitToOwner=true, maxCount=2, resolutionRule=StopOldest}
 concurrencySet [&Weapon_Gunfire_Local_Conc, &Weapon_Gunfire_Global_Conc]
 modulationSettings {
  volumeRouting=Union
  volumeModulationDestination {bEnableModulation=true, modulators [/FortMetaSoundWeapon/Items/Modulation/PP_Weapon_Fire_3p_Vol]}
 }
}
