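# Deprecated MetaSound patch that decides when a weapon fire sound finishes as a one-shot source.
# SoundsFinished or StopAutoFiring starts a delay whose time is routed from the playing wave duration or a fixed fallback value.
# That delay sets bOnAlmostFinished, then a second default-length delay fires OnFinished.
# ReTrigger resets both delays and, like OnPlay, clears the almost-finished flag.
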MetaSoundPatch /FortMetaSoundWeapon/Items/Metasounds/Subgraphs/Deprecated/MSP_Weapon_Fire_OnFinished
inputs [
 PlayingWaveDuration {type=Time}
 ReTrigger {type=Trigger}
 SoundsFinished {type=Trigger}
 StopAutoFiring {type=Trigger}
 UE.Source.OnPlay {type=Trigger}
]
interfaces [UE.Source.OneShot:1.0]
outputs [
 bOnAlmostFinished {type=Bool}
 UE.Source.OneShot.OnFinished {type=Trigger}
]
pages [Default {
  input_defaults {PlayingWaveDuration=1.5, ReTrigger=null, SoundsFinished=null, StopAutoFiring=null, UE.Source.OnPlay=null}
  nodes [
   TriggerAnyTriggerAny2_0 = "TriggerAny.Trigger Any (2)"(In 0=^SoundsFinished, In 1=^StopAutoFiring)
   TriggerRouteTriggerRouteTime2_0 = "TriggerRoute.Trigger Route (Time, 2):Time"(Value 0=^PlayingWaveDuration, Set 1=^SoundsFinished, Set 0=@TriggerAnyTriggerAny2_0.Out) {also {defaults {Value 1=1.5}}}
   TriggerDelay_0 = UE.Trigger Delay(Reset=^ReTrigger, In=@TriggerRouteTriggerRouteTime2_0.On Set, Delay Time=@TriggerRouteTriggerRouteTime2_0.Value)
   TriggerDelay_1 = UE.Trigger Delay(Reset=^ReTrigger, In=@TriggerDelay_0.Out)
   TriggerAnyTriggerAny2_1 = "TriggerAny.Trigger Any (2)"(In 0=^ReTrigger, In 1=^UE.Source.OnPlay)
   TriggerToggle_0 = UE.Trigger Toggle(Off=@TriggerAnyTriggerAny2_1.Out, On=@TriggerDelay_0.Out)
  ]
  links [{dst="@UE.Source.OneShot.OnFinished", src=TriggerDelay_1.Out}, {dst="@bOnAlmostFinished", src=TriggerToggle_0.Value}]
 }]
